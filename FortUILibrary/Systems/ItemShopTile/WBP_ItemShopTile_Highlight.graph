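# Item shop tile highlight (violator) widget, derived from WBP_UIKit_Highlight_Base, that shows a short text label styled by intensity.
# Activate and deactivate events, guarded by IsActive, play Reveal or Hide animations with sounds, interrupting the opposite animation; Construct starts it invisible.
# Hover plays a small lift animation forward or in reverse.
# Helpers map EViolatorIntensity to highlight type and font and manage size limits; SetHighlightSize is empty, and PreConstruct previews text and intensity at design time.
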WidgetBlueprint /FortUILibrary/Systems/ItemShopTile/WBP_ItemShopTile_Highlight
animations [
 Reveal {
  binding_count=1
  length_seconds=0.25
  object=&self:Reveal
  sequence {bindings [WBP_Violator_Base_C_31 {
     id=#1
     possessed_class=WBP_ItemShopTile_Highlight_C
     tracks [
      {
       type=MovieSceneFloatTrack
       display_name=Render Opacity
       sections [{type=MovieSceneFloatSection, end_frame=5, start_frame=0, channels [None {type=MovieSceneScriptingFloatChannel, keys [[0,0],[5,1]]}]}]
      }
      {
       type=MovieScene2DTransformTrack
       sections [{
         type=MovieScene2DTransformSection
         end_frame=5
         start_frame=0
         channels [
          Translation.X {type=MovieSceneScriptingFloatChannel, keys [[0,150],[5,0]]}
          Scale.X {type=MovieSceneScriptingFloatChannel, keys [[0,1.25],[5,1]]}
          Scale.Y {type=MovieSceneScriptingFloatChannel, keys [[0,0.75],[5,1]]}
          Shear.X {type=MovieSceneScriptingFloatChannel, keys [[0,-16],[5,0]]}
         ]
        }]
      }
     ]
    }]}
 }
 Hide {
  binding_count=1
  length_seconds=0.05
  object=&self:Hide
  sequence {bindings [WBP_Violator_Base_C_30 {
     id=#2
     possessed_class=WBP_ItemShopTile_Highlight_C
     tracks [{
       type=MovieSceneFloatTrack
       display_name=Render Opacity
       sections [{type=MovieSceneFloatSection, end_frame=1, start_frame=0, channels [None {type=MovieSceneScriptingFloatChannel, keys [[0,1],[1,0]]}]}]
      }]
    }]}
 }
 Hovered {
  binding_count=1
  length_seconds=0.1
  object=&self:Hovered
  sequence {bindings [WBP_Violator_Base_C_29 {
     id=#3
     possessed_class=WBP_ItemShopTile_Highlight_C
     tracks [{
       type=MovieScene2DTransformTrack
       sections [{
         type=MovieScene2DTransformSection
         end_frame=2
         start_frame=0
         channels [Translation.Y {type=MovieSceneScriptingFloatChannel, keys [[0,0],[2,-4]]}]
        }]
      }]
    }]}
 }
]
graphs [
 EventGraph {nodes [
   OnDeactivated() {is=title K2Node_CustomEvent}
   > Branch(Condition=Get IsActive() -> (IsActive:bool))
   > n1 = Branch(Condition=IsAnimationPlaying(InAnimation=Get Reveal() -> (Reveal:Widget Animation Object Reference)))
   > n2 = StopAnimation(InAnimation=Get Reveal())
   PlayAnimationForward(execute=[@n2.then, @n1.else], InAnimation=Get Hide() -> (Hide:Widget Animation Object Reference))
   > SpawnSound2D!(
    execute
    Sound:Sound Base Object Reference=Get HideSound() -> (HideSound:Sound Base Object Reference)
    VolumeMultiplier:float=1
    PitchMultiplier:float=1
    StartTime:float=0
    ConcurrencySettings:Sound Concurrency Object Reference
    bPersistAcrossLevelTransition:bool=false
    bAutoDestroy:bool=true
   ) -> (then, ReturnValue:Audio Component Object Reference)
   > Set IsActive(IsActive:bool=false) -> (Output_Get:bool)
   OnHovered() {is=title K2Node_CustomEvent}
   n3 = Get Hovered() -> (Hovered:Widget Animation Object Reference)
   > PlayAnimationForward(InAnimation=@n3)
   OnUnhovered() {is=title K2Node_CustomEvent}
   > PlayAnimationReverse(InAnimation=@n3)
   OnActivated() {is=title K2Node_CustomEvent}
   > n4 = Branch(Condition=Get IsActive())
   n5 = Branch(execute=@n4.else, Condition=IsAnimationPlaying(InAnimation=Get Hide()))
   > n6 = StopAnimation(InAnimation=Get Hide())
   PlayAnimationForward(execute=[@n6.then, @n5.else], InAnimation=Get Reveal())
   > SpawnSound2D(
    Sound=Get AppearSound() -> (AppearSound:Sound Base Object Reference)
    VolumeMultiplier=1
    PitchMultiplier=1
    StartTime=0
    bPersistAcrossLevelTransition=false
    bAutoDestroy=true
   )
   > Set IsActive(IsActive=true)
   n7 = Event PreConstruct()
   > Parent: PreConstruct(IsDesignTime=@n7.IsDesignTime)
   > Branch(Condition=@n7.IsDesignTime)
   > SetHighlightText!(execute, self, Text:Text=Get PreviewHighlightText() -> (PreviewHighlightText:Text)) -> (then)
   > SetHighlightIntensity!(
    execute
    self
    HighlightIntensity:EViolatorIntensity Enum=Get PreviewHighlightIntensity() -> (PreviewHighlightIntensity:EViolatorIntensity Enum=Low)
   ) -> (then)
   Event Construct()
   > Parent: Construct()
   > Set IsActive(IsActive=false)
   > StopAnimation(InAnimation=Get Reveal())
   > StopAnimation(InAnimation=Get Hide())
   > SetRenderOpacity(InOpacity=0)
  ]}
 InitializeHighlight {nodes [
   InitializeHighlight() {is=title K2Node_FunctionEntry}
   > Set Padding!(
    execute
    Padding:Margin Structure=Get TextPadding() -> (TextPadding:Margin Structure)
    self:Grid Slot Object Reference=SlotAsGridSlot(Widget=Get Text_Highlight() -> (Text_Highlight:Common Text Object Reference))
   ) -> (then, Output_Get:Margin Structure)
   > ClearMinDesiredHeight(self:Size Box Object Reference=Get SB_WidthLimit() -> (SB_WidthLimit:Size Box Object Reference))
   > SetSizeBox!(execute, self, MinHeight:float=Get MinHeight() -> (MinHeight:float), MaxWidth:float=Get WidthOverride() -> (WidthOverride:double)) -> (then)
  ]}
 SetHighlightIntensity {nodes [
   n1 = SetHighlightIntensity!() -> (then, HighlightIntensity:EViolatorIntensity Enum=Low) {is=title K2Node_FunctionEntry}
   > SetType!(
    execute
    self:WBP UIKit Highlight Base Object Reference
    Type:E_UI_HighlightType Enum=Select!(
     Low:E_UI_HighlightType Enum=NewEnumerator4
     Medium:E_UI_HighlightType Enum=NewEnumerator4
     High:E_UI_HighlightType Enum=NewEnumerator0
     Urgent:E_UI_HighlightType Enum=NewEnumerator7
     Index:EViolatorIntensity Enum=@n1
    ) -> (ReturnValue:E_UI_HighlightType Enum)
   ) -> (then)
   > SetFont(
    self=Get Text_Highlight()
    InFontInfo=Select(
     Low:Slate Font Info Structure=Get LowIntensityFontInfo() -> (LowIntensityFontInfo:Slate Font Info Structure)
     Medium:Slate Font Info Structure=Get MediumIntensityFontInfo() -> (MediumIntensityFontInfo:Slate Font Info Structure)
     High:Slate Font Info Structure=Get HighIntensityFontInfo() -> (HighIntensityFontInfo:Slate Font Info Structure)
     Urgent:Slate Font Info Structure=Get UrgentIntensityFontInfo() -> (UrgentIntensityFontInfo:Slate Font Info Structure)
     Index=Get HighlightIntensity() -> (HighlightIntensity:EViolatorIntensity Enum=Low)
    ) -> (ReturnValue:Slate Font Info Structure)
   )
  ]}
 SetHighlightSize {nodes [SetHighlightSize!() -> (then, FontSize:E_UI_HighlightSize Enum=NewEnumerator0) {is=title K2Node_FunctionEntry}]}
 SetHighlightText {nodes [
   n1 = SetHighlightText!() -> (then, Text:Text) {is=title K2Node_FunctionEntry}
   > SetText!(execute, self:WBP UIKit Highlight Base Object Reference, Text:Text=@n1) -> (then)
  ]}
 SetSizeBox {nodes [
   n1 = SetSizeBox!() -> (then, MinHeight:float=~, MaxWidth:float=~) {is=title K2Node_FunctionEntry}
   > Set MinHeight(MinHeight:float=@n1) -> (Output_Get:float)
   > SetHeightOverride(self=Get SB_WidthLimit(), InHeightOverride=Get MinHeight())
   > SetWidth!(execute, self:WBP UIKit Highlight Base Object Reference, NewWidthOverride:double=@n1.MaxWidth) -> (then)
  ]}
 SetWidth {nodes [
   n1 = SetWidth!() -> (then, NewWidthOverride:double) {is=title K2Node_FunctionEntry}
   > n2 = Set WidthOverride(WidthOverride:double=@n1) -> (Output_Get:double)
   > n3 = Branch(Condition="float > float"(A=@n2))
   n4 = Get SB_WidthLimit()
   > SetMaxDesiredWidth!(execute, self:Size Box Object Reference=@n4, InMaxDesiredWidth:float=@n2) -> (then)
   ClearMaxDesiredWidth(execute=@n3.else, self:Size Box Object Reference=@n4)
  ]}
 ShowDesignTimePreview {nodes [
   ShowDesignTimePreview() {is=title K2Node_FunctionEntry}
   n1 = Get Reveal()
   > PlayAnimation(InAnimation=@n1, StartAtTime=GetEndTime(self=@n1))
  ]}
]
variables [
 AppearSound {type=SoundBase, default=/Game/Sounds/Fort_UI_Sounds/FrontEnd/Whooshes/Toasts/UI_Whoosh_Tiny_Appear_Cue}
 HideSound {type=SoundBase, default=/Game/Sounds/Fort_UI_Sounds/FrontEnd/Whooshes/Toasts/UI_Whoosh_Tiny_Close_Cue}
 HighIntensityFontInfo {
  type=SlateFontInfo
  default {
   bForceMonospaced=false
   bMaterialIsStencil=false
   fontMaterial=null
   fontObject=&HeadingNow
   letterSpacing=0
   monospacedWidth=1
   size=12.75
   skewAmount=0
   typefaceFontName=76 Bold Italic
   outlineSettings {
    bApplyOutlineToDropShadows=false
    bMiteredCorners=false
    bSeparateFillAlpha=false
    outlineMaterial=null
    outlineSize=0
    outlineColor=<rgba 0 0 0 1>
   }
  }
 }
 IsActive {type=boolean}
 LowIntensityFontInfo {
  type=SlateFontInfo
  default {
   bForceMonospaced=false
   bMaterialIsStencil=false
   fontMaterial=null
   fontObject=&HeadingNow
   letterSpacing=0
   monospacedWidth=1
   size=12.75
   skewAmount=0
   typefaceFontName=76 Bold Italic
   outlineSettings {
    bApplyOutlineToDropShadows=false
    bMiteredCorners=false
    bSeparateFillAlpha=false
    outlineMaterial=null
    outlineSize=0
    outlineColor=<rgba 0 0 0 1>
   }
  }
 }
 MediumIntensityFontInfo {
  type=SlateFontInfo
  default {
   bForceMonospaced=false
   bMaterialIsStencil=false
   fontMaterial=null
   fontObject=&HeadingNow
   letterSpacing=0
   monospacedWidth=1
   size=12.75
   skewAmount=0
   typefaceFontName=76 Bold Italic
   outlineSettings {
    bApplyOutlineToDropShadows=false
    bMiteredCorners=false
    bSeparateFillAlpha=false
    outlineMaterial=null
    outlineSize=0
    outlineColor=<rgba 0 0 0 1>
   }
  }
 }
 MinHeight {type=number, default=34}
 PreviewHighlightIntensity {type=EViolatorIntensity, default=High}
 PreviewHighlightText {type=string, default="[PH] highlight"}
 TextPadding {type=Margin, default=<ltrb 12 0 12 0>}
 UrgentIntensityFontInfo {
  type=SlateFontInfo
  default {
   bForceMonospaced=false
   bMaterialIsStencil=false
   fontMaterial=null
   fontObject=&HeadingNow
   letterSpacing=0
   monospacedWidth=1
   size=12.75
   skewAmount=0
   typefaceFontName=76 Bold Italic
   outlineSettings {
    bApplyOutlineToDropShadows=false
    bMiteredCorners=false
    bSeparateFillAlpha=false
    outlineMaterial=null
    outlineSize=0
    outlineColor=<rgba 0 0 0 1>
   }
  }
 }
]
widgets {
 info {
  inheritedWidgetCount=0
  parentClass=/FortUILibrary/Systems/Highlight/WBP_UIKit_Highlight_Base.WBP_UIKit_Highlight_Base_C
  rootWidgetClass=null
  widgetCount=0
 }
 widgets []
}
