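# Item shop tile highlight component that wraps the shared FortUILibrary highlight widget in an Overlay with adjustable padding, applied on PreConstruct.
# Row state changes activate or deactivate the highlight, and hover changes, gated by CanPlayHoveredAnim, trigger its hovered or unhovered response.
# Tile dimensions set the highlight's max width minus horizontal padding, keeping its existing min height.
# It stores the offer viewmodel and exposes it through GetOfferVM, while tile anchor and widget getters return defaults.
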WidgetBlueprint /ItemShop/MultiProduct/ItemShopTileComponents/WBP_ItemShopTile_Highlight
graphs [
 EventGraph {nodes [
   Event PreConstruct()
   > UpdateHighlightPadding!(execute, self, HighlightPadding:Margin Structure=Get HighlightPadding() -> (HighlightPadding:Margin Structure)) -> (then)
   n1 = Event SetRowState!() -> (OutputDelegate, then, InRowState:E_ItemShop_RowState Enum=NewEnumerator4) {is=title K2Node_Event}
   > n2 = Switch on E_ItemShop_RowState!(execute, Selection:E_ItemShop_RowState Enum=@n1.InRowState) -> (NewEnumerator4:Exec, NewEnumerator1:Exec, NewEnumerator0:Exec, NewEnumerator2:Exec, NewEnumerator3:Exec) {is=title K2Node_SwitchEnum}
   n3 = Get Highlight() -> (Highlight:WBP Item Shop Tile Highlight Object Reference)
   OnActivated(execute=[@n2.NewEnumerator0, @n2.NewEnumerator2], self:WBP Item Shop Tile Highlight Object Reference=@n3)
   OnDeactivated(execute=[@n2.NewEnumerator1, @n2.NewEnumerator3], self:WBP Item Shop Tile Highlight Object Reference=@n3)
   n4 = Event SetTileDimensions!() -> (OutputDelegate, then, Dimensions:Vector 2D Structure) {is=title K2Node_Event}
   n5 = Get!() -> (HighlightPadding_Left:float=~, HighlightPadding_Top:float=~, HighlightPadding_Right:float=~, HighlightPadding_Bottom:float=~)
   n6 = Get Highlight()
   > SetSizeBox!(
    execute
    self:WBP Item Shop Tile Highlight Object Reference=@n6
    MinHeight:float=Get MinHeight!(self:WBP Item Shop Tile Highlight Object Reference=@n6) -> (MinHeight:float)
    MaxWidth:float=float - float(A=BreakVector2D(InVec=@n4.Dimensions), B=float + float(A=@n5, B=@n5.HighlightPadding_Right))
   ) -> (then)
   n7 = Event SetHoveredState!() -> (OutputDelegate, then, IsHovered:bool, IsImmediate:bool) {is=title K2Node_Event}
   > Branch(Condition=Get CanPlayHoveredAnim() -> (CanPlayHoveredAnim:bool))
   > n8 = Branch(Condition=@n7.IsHovered)
   n9 = Get Highlight()
   > OnHovered(self:WBP Item Shop Tile Highlight Object Reference=@n9)
   OnUnhovered(execute=@n8.else, self:WBP Item Shop Tile Highlight Object Reference=@n9)
   n10 = Event SetOfferVM!() -> (OutputDelegate, then, OfferVM:Fort Item Shop Offer VM Object Reference) {is=title K2Node_Event}
   > SetFortItemShopOfferVM!(execute, self, Viewmodel:Fort Item Shop Offer VM Object Reference=@n10.OfferVM) -> (then)
  ]}
 GetOfferVM {nodes [
   GetOfferVM() {is=title K2Node_FunctionEntry}
   > Return Node!(
    execute
    OfferVM:Fort Item Shop Offer VM Object Reference=Get FortItemShopOfferVM() -> (FortItemShopOfferVM:Fort Item Shop Offer VM Object Reference)
   )
  ]}
 GetTileAnchor {nodes [
   GetTileAnchor() {is=title K2Node_FunctionEntry}
   > Return Node!(execute, Anchor:Vector 2D Structure)
  ]}
 GetTileWidget {nodes [
   GetTileWidget() {is=title K2Node_FunctionEntry}
   > Return Node!(execute, NewParam:User Widget Object Reference)
  ]}
 UpdateHighlightPadding {nodes [
   n1 = UpdateHighlightPadding!() -> (then, HighlightPadding:Margin Structure) {is=title K2Node_FunctionEntry}
   > n2 = Set HighlightPadding(HighlightPadding:Margin Structure=@n1) -> (Output_Get:Margin Structure)
   > Set Padding!(execute, Padding:Margin Structure=@n2, self:Overlay Slot Object Reference=SlotAsOverlaySlot(Widget=Get Highlight())) -> (then, Output_Get:Margin Structure)
  ]}
]
variables [
 CanPlayHoveredAnim {type=boolean, default=true}
 HighlightPadding {type=Margin, default=<ltrb 8 8 8 8>}
]
widgets {
 info {inheritedWidgetCount=0, parentClass=&UserWidget, rootWidgetClass=&Overlay, widgetCount=2}
 widgets [
  Overlay_Highlight {widgetClassPath=&Overlay, widget=Overlay_Highlight}
  Highlight {
   widgetClassPath=/FortUILibrary/Systems/ItemShopTile/WBP_ItemShopTile_Highlight
   parent=Overlay_Highlight
   widget=Highlight
   slot {class=OverlaySlot}
  }
 ]
}
