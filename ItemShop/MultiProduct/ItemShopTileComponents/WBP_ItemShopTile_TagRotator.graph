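# Item Shop tile component widget that wraps a single enhanced tag rotator, initializing it on PreConstruct.
# Tile events pass in the offer viewmodel, cap the rotator's size box width and map row state to opacity.
# Focus follows row state on touch input and hover otherwise.
# Getters expose the offer viewmodel, tag height and tag visibility, while tile anchor and widget getters return defaults.
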WidgetBlueprint /ItemShop/MultiProduct/ItemShopTileComponents/WBP_ItemShopTile_TagRotator
graphs [
 EventGraph {nodes [
   Event PreConstruct()
   > InitializeTagRotator(self:WBP MPItem Shop Tag Rotator Base Object Reference=Get WBP_MPItemShop_TagRotator_Enhanced() -> (WBP_MPItemShop_TagRotator_Enhanced:WBP MPItem Shop Tag Rotator Enhanced Object Reference))
   n1 = Event SetRowState!() -> (OutputDelegate, then, InRowState:E_ItemShop_RowState Enum=NewEnumerator4) {is=title K2Node_Event}
   n2 = GetOwningPlayer()
   > Set Focused State!(
    execute
    self:WBP MPItem Shop Tag Rotator Base Object Reference=Get WBP_MPItemShop_TagRotator_Enhanced()
    IsFocused:bool=AND Boolean(A="Equal (Enum)"(A:E_ItemShop_RowState Enum=@n1.InRowState, B:E_ItemShop_RowState Enum=NewEnumerator0), B=IsUsingTouch(PlayerController=@n2))
    IsImmediate:bool=true
   ) -> (then) {is=title K2Node_CallFunction}
   > SetRenderOpacity(
    self=Get WBP_MPItemShop_TagRotator_Enhanced()
    InOpacity=Select!(
     NewEnumerator4:float=0
     NewEnumerator1:float=0.3
     NewEnumerator0:float=1
     NewEnumerator2:float=1
     NewEnumerator3:float=0.3
     Index:E_ItemShop_RowState Enum=@n1.InRowState
    ) -> (ReturnValue:float)
   )
   n3 = Event SetOfferVM!() -> (OutputDelegate, then, OfferVM:Fort Item Shop Offer VM Object Reference) {is=title K2Node_Event}
   > SetFortItemShopOfferVM!(execute, self, Viewmodel:Fort Item Shop Offer VM Object Reference=@n3.OfferVM) -> (then)
   n4 = Event SetTileDimensions!() -> (OutputDelegate, then, Dimensions_X:double, Dimensions_Y:double) {is=title K2Node_Event}
   > SetMaxDesiredWidth!(
    execute
    self:Size Box Object Reference=Get SizeBox_All!(self:WBP MPItem Shop Tag Rotator Base Object Reference=Get WBP_MPItemShop_TagRotator_Enhanced()) -> (SizeBox_All:Size Box Object Reference)
    InMaxDesiredWidth:float=@n4.Dimensions_X
   ) -> (then)
   n5 = Event SetHoveredState!() -> (OutputDelegate, then, IsHovered:bool, IsImmediate:bool) {is=title K2Node_Event}
   > n6 = Branch(Condition=IsUsingTouch(PlayerController=GetOwningPlayer()))
   Set Focused State(execute=@n6.else, self=Get WBP_MPItemShop_TagRotator_Enhanced(), IsFocused=@n5.IsHovered, IsImmediate=@n5.IsImmediate)
  ]}
 Get Tag Text Box Height {nodes [
   Get Tag Text Box Height() {is=title K2Node_FunctionEntry}
   n1 = Get WBP_MPItemShop_TagRotator_Enhanced()
   > Return Node!(execute, TagHeight:double=Get HeightOverride!(self:Size Box Object Reference=Get SizeBox_All(self=@n1)) -> (HeightOverride:float))
  ]}
 GetOfferVM {nodes [
   GetOfferVM() {is=title K2Node_FunctionEntry}
   > Return Node!(
    execute
    OfferVM:Fort Item Shop Offer VM Object Reference=Get FortItemShopOfferVM() -> (FortItemShopOfferVM:Fort Item Shop Offer VM Object Reference)
   )
  ]}
 GetTileAnchor {nodes [
   GetTileAnchor() {is=title K2Node_FunctionEntry}
   > Return Node!(execute, Anchor:Vector 2D Structure)
  ]}
 GetTileWidget {nodes [
   GetTileWidget() {is=title K2Node_FunctionEntry}
   > Return Node!(execute, NewParam:User Widget Object Reference)
  ]}
 IsTagTextVisible {nodes [
   IsTagTextVisible() {is=title K2Node_FunctionEntry}
   > Return Node!(
    execute
    TagIsVisible:bool=Is TagText Visible!(self:WBP MPItem Shop Tag Rotator Base Object Reference=Get WBP_MPItemShop_TagRotator_Enhanced()) -> (bTagIsVisible:bool)
   )
  ]}
]
variables []
widgets {
 info {
  inheritedWidgetCount=0
  parentClass=&UserWidget
  rootWidgetClass=/ItemShop/MultiProduct/WBP_MPItemShop_TagRotator_Enhanced.WBP_MPItemShop_TagRotator_Enhanced_C
  widgetCount=1
 }
 widgets [WBP_MPItemShop_TagRotator_Enhanced {
   widgetClassPath=/ItemShop/MultiProduct/WBP_MPItemShop_TagRotator_Enhanced
   bIsVariable=true
   widget=WBP_MPItemShop_TagRotator_Enhanced
   properties {tagHeight=36}
  }]
}
