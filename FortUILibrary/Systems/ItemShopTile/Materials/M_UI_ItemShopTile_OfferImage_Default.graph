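# Translucent UI material that draws the offer image on Item Shop tiles.
# Shared functions supply zoom, anchor and fit parameters and compute ruleset UVs, with clamped image alpha preventing edge stretching.
# Alpha combines a swipe mask, an optional trapezoid-SDF top-edge fade and tile-driven rounded corners.
# Further lerp-gated masks add rounded corners using the shop parameter collection's radius and a linear gradient fade.
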Material /FortUILibrary/Systems/ItemShopTile/Materials/M_UI_ItemShopTile_OfferImage_Default
graph {
 properties {
  bAutomaticallySetUsageInEditor=false
  bUsedWithStaticMesh=false
  blendMode=BLEND_Translucent
  materialDomain=MD_UI
  pixelDepthOffsetMode=PDOM_Legacy
  usedShadingModels=Default Lit
 }
 nodes [
  MaterialFunctionCall() {materialFunction=&GetUserInterfaceUV}
  MaterialFunctionCall!(Dimensions) {materialFunction=/FortUILibrary/MaterialFunctions/Basics/MF_UI_GetAspectRatio}
  n1 = NamedRerouteUsage()
  n2 = MaterialFunctionCall() {materialFunction=/FortUILibrary/Systems/ItemShopTile/Materials/MaterialFunctions/MF_UI_ItemShop_OfferImage_Swiping}
  n3 = NamedRerouteUsage()
  n4 = ScalarParameter() {parameterName=AddTopEdgeFade}
  n5 = VectorParameter() {group=Tile-driven params, parameterName=CornerRadiusOverride, sortPriority=40}
  n6 = MaterialFunctionCall!(
   Input Alpha=Multiply(A=Min(A=@n1, B=@n2.Swipe alpha mask), B=LinearInterpolate(B=@n3, Alpha=@n4) {constA=1})
   bOverrideCornerRadius=ScalarParameter() {group=Tile-driven params, parameterName=bOverrideCornerRadius, sortPriority=30}
   Corner Radius Override=AppendVector(A=@n5.RGB, B=@n5.RGB)
  ) {materialFunction=/FortUILibrary/Systems/ItemShopTile/Materials/MaterialFunctions/MF_UI_ItemShop_RoundedCorners}
  n7 = NamedRerouteUsage()
  n8 = NamedRerouteUsage()
  n9 = CollectionParameter() {collection=/FortUILibrary/Systems/ItemShopTile/MPC_ItemShop, parameterName=CornerRadius}
  n10 = Constant() {r=1}
  n11 = LinearInterpolate(
   A=@n6.Final Alpha
   B=Min(
    A=@n6.Final Alpha
    B=MaterialFunctionCall(UV=@n7, Size=@n8, Corner Radius=@n9, Radius in Pixels=@n10) {materialFunction=&MF_UI_CoreUI_RoundedMask}.Shape Mask
   )
   Alpha=NamedRerouteUsage()
  )
  n12 = Constant() {r=1}
  n13 = Constant() {r=0.6}
  n14 = Constant() {r=-0.25}
  Min(
   A=@n11
   B=LinearInterpolate(
    A=@n11
    B=MaterialFunctionCall(GradientStart=@n12, GradientEnd=@n13, RotationAngle=@n14) {materialFunction=&MF_UI_LinearGradient}.Result
    Alpha=NamedRerouteUsage()
   ) {constAlpha=0}
  )
  n15 = MaterialFunctionCall() {materialFunction=/FortUILibrary/Systems/ItemShopTile/Materials/MaterialFunctions/MF_UI_ItemShop_OfferImageParams}
  NamedRerouteDeclaration(None=@n15.Zoom level) {name=zoom level, nodeColor=<rgba 1 0.3516 0 1>}
  NamedRerouteDeclaration(None=@n15.X anchor point) {name=x anchor point, nodeColor=<rgba 0.0312 0 1 1>}
  NamedRerouteDeclaration(None=@n15.Fade image alpha) {name=fade image alpha, nodeColor=<rgba 0 1 0.5313 1>}
  n16 = MaterialFunctionCall!(
   UVs=Add(A=TextureCoordinate(), B=Constant2Vector() {g=-0.157})
   TopWidth=Constant() {r=0.83}
   BottomWidth=Constant() {r=1.2}
   Height=Constant() {r=1.15}
   OutlineThickness
   OutlineAlignment
   GlowMin
   GlowMax
  ) {materialFunction=/FortUILibrary/MaterialFunctions/Shapes/MF_UI_IsoscelesTrapezoid}
  n17 = Constant() {r=0.125}
  n18 = Constant() {r=0.075}
  NamedRerouteDeclaration(None=MaterialFunctionCall!("Value (0-1)"=MaterialFunctionCall(DistanceField=@n16.SDF, Threshold=@n17, Smoothing=@n18) {materialFunction=&MF_UI_AntiAliasing}.Inner) {materialFunction=/FortUILibrary/MaterialFunctions/Basics/MF_UI_EaseCurves_02_Square}.In) {name=TopFade, nodeColor=<rgba 0 0 0 1>}
  NamedRerouteDeclaration(None=@n15.Y anchor point) {name=y anchor point, nodeColor=<rgba 0.9453 1 0 1>}
  n19 = MaterialFunctionCall!(Texture Size=TextureProperty(Tex=NamedRerouteUsage()), UVs) {materialFunction=/FortUILibrary/Systems/ItemShopTile/Materials/MaterialFunctions/MF_UI_ItemShop_OfferImage_RulesetUVs}
  n20 = NamedRerouteDeclaration(None=@n19.UVs) {name=final UVs, nodeColor=<rgba 0.8281 0 1 1>}
  n21 = NamedRerouteUsage()
  NamedRerouteDeclaration(None=MaterialFunctionCall!(UVs=NamedRerouteUsage(), Alpha=TextureSample(UVs=@n20, Tex=@n21) {samplerSource=SSM_Clamp_WorldGroupSettings}.A) {desc="prevents the texture from\r\nstretching at the edges", materialFunction=/Game/UI/Foundation/Items/ClampAlphaToTransparentBorder}.Result) {name=image alpha, nodeColor=<rgba 0.4062 0 1 1>}
  NamedRerouteDeclaration(None=@n15.ScaleToFitX) {name=scale to fit X, nodeColor=<rgba 0 0.5312 1 1>}
  n22 = TextureObjectParameter() {
   group=Tile-driven params
   parameterName=OfferImage
   sortPriority=0
   texture=/Game/Athena/UI/Challenges/Art/TileImages/T_UI_ChallengeTile_Jonesy
  }
  NamedRerouteDeclaration(None=@n22) {name=texture, nodeColor=<rgba 0.1953 1 0 1>}
  NamedRerouteDeclaration(None=@n15.HasRoundedCorners) {name=has rounded corners, nodeColor=<rgba 0.8516 1 0 1>}
  NamedRerouteDeclaration(None=@n19.UV position) {name=anchored UVs, nodeColor=<rgba 0 1 0.0156 1>}
  NamedRerouteDeclaration(None=@n19.UV scale) {name=UV scale, nodeColor=<rgba 1 0 0.3047 1>}
 ]
}
