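# UI-domain material, named for the font of bottom tags in the multi-product item shop.
# Its only logic takes the RGB of a Color vector parameter, white by default, and multiplies it by a constant that darkens it to a fifth.
# The graph shows no textures and no link from that multiply to a material output.
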Material /ItemShop/MultiProduct/Art/M_UI_ItemShop_BottomTags_Font
graph {
 properties {bAutomaticallySetUsageInEditor=false, bUsedWithStaticMesh=false, materialDomain=MD_UI, usedShadingModels=Default Lit}
 nodes [Multiply(A=VectorParameter() {parameterName=Color, defaultValue=<rgba 1 1 1 1>}.RGB) {constB=0.2}]
}
