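# UI material for the bottom tag on Item Shop multi-product tiles, alpha-composited with rounded corners and animated by an IsHovered parameter.
# Two angled box masks shift with hover on aspect-corrected, right-aligned UVs, alongside a skewed plus icon that slides horizontally.
# Two panned screen-space noise textures build a light shimmer that brightens on hover.
# The combined alpha passes through the shared corner function using MPC_ItemShop's CornerRadius, scaled by vertex alpha.
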Material /ItemShop/MultiProduct/Art/M_UI_ShopTile_BottomTags
graph {
 properties {
  bAutomaticallySetUsageInEditor=false
  bUsedWithStaticMesh=false
  blendMode=BLEND_AlphaComposite
  materialDomain=MD_UI
  usedShadingModels=Default Lit
 }
 nodes [
  LinearInterpolate() {constA=0.065, constB=0.13}
  LinearInterpolate() {constA=0.35}
  n1 = TextureCoordinate()
  n2 = Constant2Vector() {g=40, r=100}
  n3 = Divide(
   A=ComponentMask(None=MaterialFunctionCall() {materialFunction=&GetUserInterfaceUV}.Pixel Size) {g=true}
   B=Multiply(A=@n2.G, B=MaterialFunctionCall() {materialFunction=/FortUILibrary/MaterialFunctions/Basics/MF_UI_ViewportScale}.Viewport Scale Factor)
  )
  n4 = AppendVector(A=Constant() {r=1}, B=@n3)
  n5 = Constant()
  n6 = Subtract(A=@n3)
  n7 = CollectionParameter() {collection=/FortUILibrary/Systems/ItemShopTile/MPC_ItemShop, parameterName=CornerRadius}
  n8 = Constant4Vector() {constant=<rgba 0 1 0 1>}
  n9 = MaterialFunctionCall(R=@n8.R, G=@n8.R, B=@n8.R, A=@n8.R) {materialFunction=&AppendMany}
  MaterialFunctionCall(
   UV=Subtract(A=Multiply(A=@n1, B=@n4) {constB=0.5}, B=AppendVector(A=@n5, B=@n6)) {constB=0.25}
   Corner Radius=Multiply(A=@n7, B=Multiply(A=@n9.RGBA, B=@n3))
   Radius in Pixels=Constant() {r=1}
   Corners Modulated by Size=Constant()
   Pixel Size=@n2
  ) {materialFunction=&MF_UI_CoreUI_RoundedMask}
  MaterialFunctionCall(Number=@n3) {materialFunction=&DebugScalarValues}
  n10 = NamedRerouteUsage()
  n11 = Subtract(A=NamedRerouteUsage()) {constB=0.09}
  n12 = ScalarParameter() {defaultValue=1, parameterName=IsHovered, sliderMax=1}
  n13 = Add(A=@n10, B=AppendVector(A=Constant(), B=LinearInterpolate(A=@n11, Alpha=@n12) {constA=-0.64, constB=-0.1}))
  n14 = LinearInterpolate(Alpha=ScalarParameter() {defaultValue=1, parameterName=IsHovered, sliderMax=1}) {constA=-20, constB=0}
  n15 = Constant() {r=1}
  n16 = Constant()
  n17 = LinearInterpolate(
   A=NamedRerouteDeclaration(None=Constant() {r=-0.6}) {name=BoxAngleXOffset, nodeColor=<rgba 0 0 0 1>}
   Alpha=ScalarParameter() {defaultValue=1, parameterName=IsHovered, sliderMax=1}
  ) {constA=-0.55, constB=0}
  n18 = MaterialFunctionCall!(UV=Add(A=@n10, B=AppendVector(A=@n16, B=@n17)), Corner, Angle=@n14, Flip, Falloff) {materialFunction=/FortUILibrary/MaterialFunctions/Shapes/MF_UI_BoxAngle}
  n19 = Multiply(A=Min(
    A=MaterialFunctionCall(UV=@n13, Angle=@n14, Flip=@n15) {materialFunction=&MF_UI_BoxAngle}.Result
    B=@n18.Result
   )) {constB=0.007}
  n20 = Multiply(A=NamedRerouteUsage(), B=@n18.Result)
  n21 = Add(A=Multiply(A=NamedRerouteUsage()), B=Saturate(None=Add(A=@n19, B=@n20)))
  n22 = Constant() {r=1}
  n23 = Multiply(A=@n7, B=@n9.RGBA)
  Multiply(
   A=VertexColor().A
   B=Multiply(
    A=Constant3Vector() {constant=<rgba 1 1 1 0>}
    B=MaterialFunctionCall!(Input Alpha=@n21, bOverrideCornerRadius=@n22, Corner Radius Override=@n23) {materialFunction=/FortUILibrary/Systems/ItemShopTile/Materials/MaterialFunctions/MF_UI_ItemShop_RoundedCorners}.Final Alpha
   )
  )
  n24 = MaterialFunctionCall!(ScreenUV, TextureSize, ScalePivotPoint) {materialFunction=/FortUILibrary/MaterialFunctions/Basics/MF_UI_ScreenSpaceUVScaling}
  n25 = MaterialFunctionCall(
   UVs=MaterialFunctionCall(UV=@n24.Scaled to Fill) {materialFunction=&MF_UI_UVScaling}.Scaled To Fill
   "Rotation Angle (0-1)"=Constant() {r=-0.17}
  ) {materialFunction=&CustomRotator}
  n26 = TextureSample(UVs=MaterialFunctionCall(UVs=Panner(Coordinate=@n25.Rotated Values) {speedX=-0.003, speedY=0.02}, Texture Scale=Constant2Vector() {g=0.7, r=0.1}) {materialFunction=&ScaleUVsByCenter}.UVs) {samplerType=SAMPLERTYPE_Alpha, texture=/FortUILibrary/Materials/Background/Nebulous/T_UI_Nebulous_Noise}
  n27 = Constant() {r=-0.07}
  n28 = ScalarParameter() {defaultValue=1, parameterName=IsHovered, sliderMax=1}
  n29 = MaterialFunctionCall(
   UVs=Panner(Coordinate=MaterialFunctionCall(UVs=@n24.ScaleY, "Rotation Angle (0-1)"=@n27) {materialFunction=&CustomRotator}.Rotated Values) {speedX=-0.02, speedY=0.03}
   Texture Scale=AppendVector(A=LinearInterpolate(Alpha=@n28) {constA=0.05, constB=0.5}, B=Constant() {r=1})
  ) {materialFunction=&ScaleUVsByCenter}
  n30 = Saturate(None=Add(
    A=LinearInterpolate(Alpha=@n26.RGB) {constA=0.2, constB=0.35}
    B=TextureSample(UVs=@n29.UVs) {samplerType=SAMPLERTYPE_Grayscale, texture=/Game/UIKit/Buttons/Textures/LowResBlurredNoise}.RGB
   ))
  n31 = LinearInterpolate(Alpha=@n30) {constA=0.2, constB=0.8}
  NamedRerouteDeclaration(None=LinearInterpolate(
    A=LinearInterpolate(Alpha=@n30) {constA=0.004, constB=0.009}
    B=MaterialFunctionCall!("Value (0-1)"=@n31) {materialFunction=/FortUILibrary/MaterialFunctions/Basics/MF_UI_EaseCurves_01_Sine}.Out
    Alpha=ScalarParameter() {defaultValue=1, parameterName=IsHovered, sliderMax=1}
   )) {name=LightShimmer, nodeColor=<rgba 0 0 0 1>}
  n32 = MaterialFunctionCall("Value (0-1)"=ScalarParameter() {defaultValue=1, parameterName=IsHovered, sliderMax=1}) {materialFunction=&MF_UI_EaseCurves_01_Sine}
  n33 = Add(
   A=MaterialFunctionCall!(UV=NamedRerouteUsage(), Skew Angle=Constant2Vector() {r=12}) {materialFunction=/FortUILibrary/MaterialFunctions/Basics/MF_UI_Skew}.Skewed UVs
   B=AppendVector(A=LinearInterpolate(Alpha=@n32.In-Out) {constA=0.07, constB=-1.5}, B=Constant())
  )
  n34 = MaterialFunctionCall!(
   UVs=MaterialFunctionCall(UVs=@n33, Texture Scale=Constant() {r=1}) {materialFunction=&ScaleUVsByCenter}.UVs
   Height=Constant() {r=0.67}
   InnerThickness
   Stroke
   Corner Roundness=Constant() {r=0.15}
  ) {materialFunction=/FortUILibrary/Systems/Items/Materials/MF_UI_PlusShape}
  n35 = Constant() {r=0.55}
  n36 = MaterialFunctionCall(DistanceField=MaterialFunctionCall(
    UVs=MaterialFunctionCall(UVs=@n33, Texture Scale=@n35) {materialFunction=&ScaleUVsByCenter}.UVs
    Height=Constant() {r=0.98}
    InnerThickness=Constant() {r=0.27}
    Corner Roundness=Constant() {r=0.2}
   ) {materialFunction=&MF_UI_PlusShape}.SDF) {materialFunction=&MF_UI_AntiAliasing}
  NamedRerouteDeclaration(None=Subtract(
    A=MaterialFunctionCall(DistanceField=@n34.SDF) {materialFunction=&MF_UI_AntiAliasing}.Inner
    B=NamedRerouteDeclaration(None=@n36.Inner) {name=Plus Icon Fill, nodeColor=<rgba 0 0 0 1>}
   )) {name=Plus Icon Stroke, nodeColor=<rgba 0 0 0 1>}
  n37 = TextureCoordinate()
  n38 = MaterialFunctionCall!(Dimensions) {materialFunction=/FortUILibrary/MaterialFunctions/Basics/MF_UI_GetAspectRatio}
  NamedRerouteDeclaration(None=Subtract(A=Multiply(A=@n37, B=@n38.Max Ratio), B=Subtract(A=@n38.Max Ratio)) {constB=0}) {name=RightAlignedUVs, nodeColor=<rgba 0 0 0 1>}
 ]
}
