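# Translucent UI material for the bottom tag area of an Item Shop tile, by name a simple variant.
# A step on vertical UV against the inverted IsHovered parameter drives a bottom-up alpha wipe, reaching partial opacity when hovered.
# That alpha feeds the shared rounded-corners function.
# Its radius override multiplies the shop collection's CornerRadius by a vector built solely from a constant's zero red channel.
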Material /ItemShop/MultiProduct/Art/M_UI_ShopTile_BottomTags_Simple
graph {
 properties {bAutomaticallySetUsageInEditor=false, blendMode=BLEND_Translucent, materialDomain=MD_UI, usedShadingModels=Default Lit}
 nodes [
  Constant3Vector()
  n1 = OneMinus(None=ScalarParameter() {defaultValue=1, parameterName=IsHovered, sliderMax=1})
  n2 = ComponentMask(None=TextureCoordinate()) {g=true}
  n3 = Constant4Vector() {constant=<rgba 0 1 0 1>}
  MaterialFunctionCall!(
   Input Alpha=LinearInterpolate(Alpha=Step(Y=@n1, X=@n2) {constX=0.5}) {constB=0.42}
   bOverrideCornerRadius=Constant() {r=1}
   Corner Radius Override=Multiply(
    A=CollectionParameter() {collection=/FortUILibrary/Systems/ItemShopTile/MPC_ItemShop, parameterName=CornerRadius}
    B=MaterialFunctionCall(R=@n3.R, G=@n3.R, B=@n3.R, A=@n3.R) {materialFunction=&AppendMany}.RGBA
   )
  ) {materialFunction=/FortUILibrary/Systems/ItemShopTile/Materials/MaterialFunctions/MF_UI_ItemShop_RoundedCorners}
 ]
}
