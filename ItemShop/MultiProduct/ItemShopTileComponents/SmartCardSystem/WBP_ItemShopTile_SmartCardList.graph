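# Item shop tile widget that lists an offer's secondary granted items as wrapped smart cards.
# PopulateList picks vehicle ruleset and wrap mode, sizes a SizeBox from card size, spacing and grid, and fills a dynamic entry box, turning the last visible card into an overflow count.
# Small tiles without an active filter collapse to an item-count tag.
# While filtering, items come from the section's filter, and width updates drive a single-row column count.
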WidgetBlueprint /ItemShop/MultiProduct/ItemShopTileComponents/SmartCardSystem/WBP_ItemShopTile_SmartCardList
graphs [
 EventGraph {nodes [
   n1 = Event SetRowState!() -> (OutputDelegate, then, InRowState:E_ItemShop_RowState Enum=NewEnumerator4) {is=title K2Node_Event}
   > Set RenderOpacity!(
    execute
    RenderOpacity:float=Select!(
     NewEnumerator4:float=0
     NewEnumerator1:float=0.3
     NewEnumerator0:float=1
     NewEnumerator2:float=1
     NewEnumerator3:float=0.3
     Index:E_ItemShop_RowState Enum=@n1.InRowState
    ) -> (ReturnValue:float)
    self:Widget Object Reference=Get Overlay_All() -> (Overlay_All:Overlay Object Reference)
   ) -> (then, Output_Get:float)
   n2 = Event SetSectionID!() -> (OutputDelegate, then, InSectionID:String) {is=title K2Node_Event}
   > Branch(Condition="Not Equal, Case Insenstive (String)"!(A:String=@n2.InSectionID, B:String=Get SectionID() -> (SectionID:String)) -> (ReturnValue:bool))
   > Set SectionID(SectionID:String=@n2.InSectionID) -> (Output_Get:String)
   > n3 = Get!(execute:Exec) -> (then:Exec, else:Exec, FortItemShopOfferVM:Fort Item Shop Offer VM Object Reference)
   n4 = Get FilterVM!(self:Fort Item Shop VM Object Reference=Get FortItemShopVM() -> (FortItemShopVM:Fort Item Shop VM Object Reference)) -> (FilterVM:Fort Item Shop Filter VM Object Reference)
   > PopulateList!(
    execute
    self
    ItemsToShow:Array of Fort Item VM Object References=GetItemsToShow!(
     self
     GrantedItems:Array of Fort Item VM Object References=Get GrantedItems!(self:Fort Item Shop Offer VM Object Reference=@n3) -> (GrantedItems:Array of Fort Item VM Object References)
     SecondaryGrantedItems:Array of Fort Item VM Object References=Get DisplayedSecondaryGrantedItems!(self:Fort Item Shop Offer VM Object Reference=@n3) -> (DisplayedSecondaryGrantedItems:Array of Fort Item VM Object References)
     IsFilterActive:bool=Get bIsActive!(self:Fort Item Shop Filter VM Object Reference=@n4) -> (bIsActive:bool)
     TileSize:EItemShopTileSize Enum=Get TileSize!(self:Fort Item Shop Offer VM Object Reference=@n3) -> (TileSize:EItemShopTileSize Enum=Size_1_x_1)
    ) -> (FilteredItems:Array of Fort Item VM Object References)
   ) -> (then)
   n5 = Event SetOfferVM!() -> (OutputDelegate, then, OfferVM:Fort Item Shop Offer VM Object Reference) {is=title K2Node_Event}
   > SetFortItemShopOfferVM!(execute, self, Viewmodel:Fort Item Shop Offer VM Object Reference=@n5.OfferVM) -> (then)
   n6 = Event PreConstruct()
   > n7 = Branch(Condition=@n6.IsDesignTime)
   > n8 = SetIsCollapsedMode!(execute, self, IsCollapsed:bool=Get IsCollapsedMode() -> (IsCollapsedMode:bool)) -> (then)
   n9 = Get WBP_UIKit_Tag() -> (WBP_UIKit_Tag:WBP UIKit Tag Object Reference)
   SetTagBackgroundColor!(
    execute=[@n8.then, @n7.else]
    self:WBP UIKit Tag Base Object Reference=@n9
    Color:Linear Color Structure=(R=0.0513,G=0.0513,B=0.0513,A=0.502)
   ) -> (then)
   > SetTagTextColor!(execute, self:WBP UIKit Tag Base Object Reference=@n9, Color:Linear Color Structure=(R=1,G=1,B=1,A=1)) -> (then)
  ]}
 GetItemsToShow {nodes [
   n1 = GetItemsToShow!() -> (
    then
    GrantedItems:Array of Fort Item VM Object References
    SecondaryGrantedItems:Array of Fort Item VM Object References
    IsFilterActive:bool=~
    TileSize:EItemShopTileSize Enum=Size_1_x_1
   ) {is=title K2Node_FunctionEntry}
   > n2 = GetSecondaryGrantedItemsForTile!(
    execute
    FilterVM:Fort Item Shop Filter VM Object Reference=Get FilterVM(self=Get FortItemShopVM())
    SectionId:Name=String To Name(InString=Get SectionID())
    InOfferSecondaryGrantedItems:Array of Fort Item VM Object References=@n1
   ) -> (then, ReturnValue:Array of Fort Item VM Object References)
   > Return Node!(
    execute
    FilteredItems:Array of Fort Item VM Object References=Select!(
     Option 0:Array of Fort Item VM Object References=@n1.SecondaryGrantedItems
     Option 1:Array of Fort Item VM Object References=@n2
     Index:bool=@n1.IsFilterActive
    ) -> (ReturnValue:Array of Fort Item VM Object References)
   )
  ]}
 GetOfferVM {nodes [
   GetOfferVM() {is=title K2Node_FunctionEntry}
   > Return Node!(
    execute
    OfferVM:Fort Item Shop Offer VM Object Reference=Get FortItemShopOfferVM() -> (FortItemShopOfferVM:Fort Item Shop Offer VM Object Reference)
   )
  ]}
 GetSmartCardSize {nodes [
   n1 = GetSmartCardSize!() -> (then, UsesVehicleRuleset:bool=~) {is=title K2Node_FunctionEntry}
   n2 = IsMobileGame()
   n3 = Get SmartCardSize_Vehicle() -> (SmartCardSize_Vehicle:int=0)
   n4 = Get SmartCardSizeMobile_Vehicle() -> (SmartCardSizeMobile_Vehicle:int=0)
   n5 = IsMobileGame()
   n6 = Get SmartCardSizeMobile() -> (SmartCardSizeMobile:int=0)
   n7 = Get SmartCardSize() -> (SmartCardSize:int=0)
   > Return Node!(execute, ReturnValue:int=SelectInt(A=SelectInt(A=@n4, B=@n3, bPickA=@n2), B=SelectInt(A=@n6, B=@n7, bPickA=@n5), bPickA=@n1))
  ]}
 GetTileAnchor {nodes [
   GetTileAnchor() {is=title K2Node_FunctionEntry}
   > Return Node!(execute, Anchor:Vector 2D Structure)
  ]}
 GetTileWidget {nodes [
   GetTileWidget() {is=title K2Node_FunctionEntry}
   > Return Node!(execute, NewParam:User Widget Object Reference)
  ]}
 IsNumberTile {nodes [
   IsNumberTile!() -> (then, EntryIndex:int, VisibleCards:int, NumItems:int) {is=title K2Node_FunctionEntry}
   n1 = int - int(A=Get VisibleCards() -> (VisibleCards:int=0))
   n2 = Get EntryIndex() -> (EntryIndex:int=0)
   n3 = Get NumItems() -> (NumItems:int=0)
   n4 = Get VisibleCards() -> (VisibleCards=0)
   > Return Node!(execute, IsNumberTile:bool=AND Boolean(A="Equal (Integer)"(A=@n2, B=@n1), B="integer > integer"(A=@n3, B=@n4)))
  ]}
 PopulateList {nodes [
   PopulateList!() -> (then, ItemsToShow:Array of Fort Item VM Object References) {is=title K2Node_FunctionEntry}
   > Set NumItems(NumItems:int=Length(TargetArray:Array of Fort Item VM Object References=Get ItemsToShow() -> (ItemsToShow:Array of Fort Item VM Object References))) -> (Output_Get:int=0)
   n1 = Get FortItemShopVM()
   > Set IsFilteringActive(IsFilteringActive:bool=Get bIsActive(self=Get FilterVM(self=@n1))) -> (Output_Get:bool)
   > Set VisibleCards(VisibleCards:int=GetNumVisibleSmartCards!(
     ItemsArray:Array of Fort Item VM Object References=Get ItemsToShow()
     OfferVM:Fort Item Shop Offer VM Object Reference=Get FortItemShopOfferVM()
     GridSizeOverride:Int Point Structure=Get MaxGridSizeOverride() -> (MaxGridSizeOverride:Int Point Structure)
     IsFilteringActive:bool=Get IsFilteringActive() -> (IsFilteringActive:bool)
    ) -> (ReturnValue:int=0)) -> (Output_Get:int=0)
   n2 = Get FortItemShopOfferVM()
   n3 = Get IsFilteringActive()
   n4 = Get IsUsedInBillboard() -> (IsUsedInBillboard:bool)
   > Set UsesVehicleRuleset(UsesVehicleRuleset:bool=AND Boolean!(
     A=ShouldUseVehicleRuleset!(OfferVM:Fort Item Shop Offer VM Object Reference=@n2) -> (ReturnValue:bool)
     B=NOT Boolean(A=@n4)
     C:bool=NOT Boolean(A=@n3)
    ) -> (ReturnValue)) -> (Output_Get:bool)
   n5 = Get EntryBoxWrapOverride() -> (EntryBoxWrapOverride:EntryBoxWrapModeOverride Enum=NewEnumerator2)
   n6 = Get UsesVehicleRuleset() -> (UsesVehicleRuleset:bool)
   > n7 = Set EntryBoxWrapMode(EntryBoxWrapMode:EntryBoxWrapModeOverride Enum=Select(
     Option 0:EntryBoxWrapModeOverride Enum=@n5
     Option 1:EntryBoxWrapModeOverride Enum=Select(Option 0=NewEnumerator1, Option 1=NewEnumerator0, Index=@n6)
     Index="Equal (Enum)"(A:EntryBoxWrapModeOverride Enum=@n5, B:EntryBoxWrapModeOverride Enum=NewEnumerator2)
    ) -> (ReturnValue:EntryBoxWrapModeOverride Enum)) -> (Output_Get:EntryBoxWrapModeOverride Enum=NewEnumerator2)
   n8 = Get UsesVehicleRuleset()
   n9 = Get TileSize(self=Get FortItemShopOfferVM()) -> (TileSize=Size_1_x_1)
   n10 = Get IsFilteringActive()
   n11 = Get!() -> (MaxGridSizeOverride_X:int, MaxGridSizeOverride_Y:int)
   > SetupSizeBox!(
    execute
    self
    WrapMode:EntryBoxWrapModeOverride Enum=@n7
    UsesVehicleRuleset:bool=Get UsesVehicleRuleset()
    GridSize:Int Point Structure=Select(
     Option 0:Int Point Structure=GetMaxSmartCardGridSize!(TileSize:EItemShopTileSize Enum=@n9, UsingVehicleRuleset:bool=@n8, IsFilteringActive:bool=@n10) -> (GridSize:Int Point Structure)
     Option 1:Int Point Structure=Get MaxGridSizeOverride()
     Index="integer > integer"(A=@n11)
    ) -> (ReturnValue:Int Point Structure)
   ) -> (then)
   > Reset!(
    execute
    self:Dynamic Entry Box Object Reference=[
     Get DynamicEntryBox_SmartCards_Vertical() -> (DynamicEntryBox_SmartCards_Vertical:Fort Dynamic Entry Box Object Reference)
     Get DynamicEntryBox_SmartCards_Horizontal() -> (DynamicEntryBox_SmartCards_Horizontal:Fort Dynamic Entry Box Object Reference)
    ]
    bDeleteWidgets:bool=false
   ) -> (then)
   > SetupDynamicEntryBox!(
    execute
    self
    IsVehicleRuleset:bool=Get UsesVehicleRuleset()
    WrapMode:EntryBoxWrapModeOverride Enum=Get EntryBoxWrapMode() -> (EntryBoxWrapMode:EntryBoxWrapModeOverride Enum=NewEnumerator2)
   ) -> (then)
   > SetIsCollapsedMode(IsCollapsed=ShouldShowCollapsedMode!(self) -> (ShouldCollapse:bool))
   > n12 = Branch(Condition=Get IsCollapsedMode())
   n13 = Get GrantedItems(self=Get FortItemShopOfferVM())
   > n14 = Branch(Condition="integer <= integer"(A=Length(TargetArray=@n13), B=1))
   n15 = Get WBP_UIKit_Tag()
   > Set Visibility(Visibility=Collapsed, self=@n15)
   Set Visibility(execute=@n14.else, Visibility=SelfHitTestInvisible, self=@n15)
   n16 = Get ItemsToShow()
   > SetTagText!(
    execute
    self:WBP UIKit Tag Base Object Reference=Get WBP_UIKit_Tag()
    Text:Text=Format Text!(Format="LOCTEXT(\"{numItems} items\")", numItems:int=Length(TargetArray=@n16)) -> (Result)
   ) -> (then)
   n17 = For Each Loop(Exec=@n12.else, Array:Array of Fort Item VM Object References=Get ItemsToShow()) -> (Array Element:Fort Item VM Object Reference)
   > Set LoopItem(LoopItem:Fort Item VM Object Reference=@n17) -> (Output_Get:Fort Item VM Object Reference)
   > Set EntryIndex(EntryIndex:int=@n17.Array Index) -> (Output_Get:int=0)
   > Branch(Condition="integer < integer"(A=@n17.Array Index, B=Get VisibleCards() -> (VisibleCards=0)))
   > n18 = Create Entry!(
    execute
    self:Dynamic Entry Box Object Reference=Select!(
     NewEnumerator2:Fort Dynamic Entry Box Object Reference
     NewEnumerator0:Fort Dynamic Entry Box Object Reference=Get DynamicEntryBox_SmartCards_Horizontal()
     NewEnumerator1:Fort Dynamic Entry Box Object Reference=Get DynamicEntryBox_SmartCards_Vertical()
     Index:EntryBoxWrapModeOverride Enum=Get EntryBoxWrapMode() -> (EntryBoxWrapMode=NewEnumerator2)
    ) -> (ReturnValue:Fort Dynamic Entry Box Object Reference)
   ) -> (then, ReturnValue:User Widget Object Reference)
   > n19 = Cast To WBP_ItemShopTile_SmartCard!(execute, Object=@n18) -> (then, CastFailed, AsWBP Item Shop Tile Smart Card:WBP Item Shop Tile Smart Card Object Reference) {is=title K2Node_DynamicCast}
   > n20 = Set LoopCard(LoopCard:WBP Item Shop Tile Smart Card Object Reference=@n19) -> (Output_Get:WBP Item Shop Tile Smart Card Object Reference)
   > TweakVehicleRulesetVisuals!(execute, self, Card:WBP Item Shop Tile Smart Card Object Reference=@n20, UsesVehicleRuleset:bool=Get UsesVehicleRuleset()) -> (then)
   > n21 = Branch(Condition=IsNumberTile!(
     self
     EntryIndex:int=Get EntryIndex() -> (EntryIndex=0)
     VisibleCards:int=Get VisibleCards() -> (VisibleCards=0)
     NumItems:int=Get NumItems() -> (NumItems=0)
    ) -> (IsNumberTile:bool) {is=title K2Node_CallFunction})
   n22 = Get VisibleCards() -> (VisibleCards=0)
   > SetIsNumberTile!(
    execute
    self:WBP Item Shop Tile Smart Card Object Reference=Get LoopCard() -> (LoopCard:WBP Item Shop Tile Smart Card Object Reference)
    IsNumberTile:bool=true
    Number:int=int - int(A=Get NumItems() -> (NumItems=0), B=int - int(A=@n22))
   ) -> (then)
   SetFortItemDetailsVM!(
    execute=@n21.else
    self:WBP Item Shop Tile Smart Card Object Reference=Get LoopCard()
    Viewmodel:Fort Item Details VM Object Reference=Get ItemDetails!(self:Fort Item VM Object Reference=Get LoopItem() -> (LoopItem:Fort Item VM Object Reference)) -> (ItemDetails:Fort Item Details VM Object Reference)
   ) -> (then)
   > SetIsNumberTile(self=Get LoopCard(), IsNumberTile=false, Number=0)
  ]}
 SetIsCollapsedMode {nodes [
   n1 = SetIsCollapsedMode!() -> (then, IsCollapsed:bool=~) {is=title K2Node_FunctionEntry}
   > Set IsCollapsedMode(IsCollapsedMode:bool=@n1) -> (Output_Get:bool)
   > SetActiveWidget(
    self=Get Switcher_Mode() -> (Switcher_Mode:Common Visibility Switcher Object Reference)
    Widget=Select!(
     Option 0:Widget Object Reference=Get SizeBox_All() -> (SizeBox_All:Size Box Object Reference)
     Option 1:Widget Object Reference=Get WBP_UIKit_Tag()
     Index:bool=Get IsCollapsedMode()
    ) -> (ReturnValue:Widget Object Reference)
   )
  ]}
 SetMaxHeight {nodes [
   n1 = SetMaxHeight!() -> (then, NumRows:int, UsesVehicleRuleset:bool=~) {is=title K2Node_FunctionEntry}
   n2 = float * float(
    A:int=int - int(A=@n1)
    B=Get!(self:Dynamic Entry Box Base Object Reference=Get DynamicEntryBox_SmartCards_Vertical()) -> (EntrySpacing_X:double, EntrySpacing_Y:double).EntrySpacing_Y
   ) -> (ReturnValue:int) {is=title K2Node_PromotableOperator}
   n3 = int * int(A=@n1, B=GetSmartCardSize!(UsesVehicleRuleset:bool=@n1.UsesVehicleRuleset) -> (ReturnValue:int=0) {is=title K2Node_CallFunction})
   > Set HeightOverride!(execute, HeightOverride:float="To Float (Integer)"(InInt=int + int(A=@n3, B=@n2)), self:Size Box Object Reference=Get SizeBox_All()) -> (then, Output_Get:float)
  ]}
 SetMaxWidth {nodes [
   n1 = SetMaxWidth!() -> (then, NumColumns:int, UsesVehicleRuleset:bool=~) {is=title K2Node_FunctionEntry}
   n2 = float * float(A=int - int(A=@n1), B=Get(self=Get DynamicEntryBox_SmartCards_Vertical()))
   n3 = int * int(A=@n1, B=GetSmartCardSize(UsesVehicleRuleset=@n1.UsesVehicleRuleset) -> (ReturnValue=0))
   > Set WidthOverride!(execute, WidthOverride:float="To Float (Integer)"(InInt=int + int(A=@n3, B=@n2)), self:Size Box Object Reference=Get SizeBox_All()) -> (then, Output_Get:float)
  ]}
 SetupDynamicEntryBox {nodes [
   SetupDynamicEntryBox!() -> (then, IsVehicleRuleset:bool=~, WrapMode:EntryBoxWrapModeOverride Enum=NewEnumerator2) {is=title K2Node_FunctionEntry}
   > SetActiveWidget(
    self=Get Switcher_EntryBoxWrapMode() -> (Switcher_EntryBoxWrapMode:Common Visibility Switcher Object Reference)
    Widget=Select!(
     NewEnumerator2:Fort Dynamic Entry Box Object Reference
     NewEnumerator0:Fort Dynamic Entry Box Object Reference=Get DynamicEntryBox_SmartCards_Horizontal()
     NewEnumerator1:Fort Dynamic Entry Box Object Reference=Get DynamicEntryBox_SmartCards_Vertical()
     Index:EntryBoxWrapModeOverride Enum=Get WrapMode() -> (WrapMode:EntryBoxWrapModeOverride Enum=NewEnumerator2)
    ) -> (ReturnValue:Fort Dynamic Entry Box Object Reference)
   )
   > SetEntrySpacing!(
    execute
    self:Dynamic Entry Box Base Object Reference=[Get DynamicEntryBox_SmartCards_Horizontal(), Get DynamicEntryBox_SmartCards_Vertical()]
    InEntrySpacing:"Vector 2D Structure (by ref)"=Select!(
     Option 0:Vector 2D Structure=Get EntrySpacingDefault() -> (EntrySpacingDefault:Vector 2D Structure)
     Option 1:Vector 2D Structure=Get EntrySpacingVehicle() -> (EntrySpacingVehicle:Vector 2D Structure)
     Index:bool=Get IsVehicleRuleset() -> (IsVehicleRuleset:bool)
    ) -> (ReturnValue:Vector 2D Structure)
   ) -> (then)
  ]}
 SetupSizeBox {nodes [
   n1 = SetupSizeBox!() -> (then, WrapMode:EntryBoxWrapModeOverride Enum=NewEnumerator2, UsesVehicleRuleset:bool=~, GridSize:Int Point Structure) {is=title K2Node_FunctionEntry}
   > n2 = Switch on EntryBoxWrapModeOverride!(execute, Selection:EntryBoxWrapModeOverride Enum=@n1) -> (NewEnumerator2:Exec, NewEnumerator0:Exec, NewEnumerator1:Exec) {is=title K2Node_SwitchEnum}
   ClearHeightOverride(execute=@n2.NewEnumerator0, self:Size Box Object Reference=Get SizeBox_All())
   > SetMaxWidth!(execute, self, NumColumns:int=Get!() -> (GridSize_X:int, GridSize_Y:int), UsesVehicleRuleset:bool=Get UsesVehicleRuleset()) -> (then) {is=title K2Node_CallFunction}
   ClearWidthOverride(execute=@n2.NewEnumerator1, self:Size Box Object Reference=Get SizeBox_All())
   > SetMaxHeight!(execute, self, NumRows:int=Get().GridSize_Y, UsesVehicleRuleset:bool=Get UsesVehicleRuleset()) -> (then) {is=title K2Node_CallFunction}
  ]}
 ShouldShowCollapsedMode {nodes [
   ShouldShowCollapsedMode() {is=title K2Node_FunctionEntry}
   n1 = Get bIsActive(self=Get FilterVM(self=Get FortItemShopVM()))
   n2 = Get TileSize(self=Get FortItemShopOfferVM()) -> (TileSize=Size_1_x_1)
   n3 = "Equal (Enum)"(A:EItemShopTileSize Enum=@n2, B:EItemShopTileSize Enum=Size_1_x_1)
   n4 = "Equal (Enum)"(A=@n2, B=Size_1_x_2)
   > Return Node!(execute, ShouldCollapse:bool=AND Boolean!(A=NOT Boolean(A=@n1), B=OR Boolean(A=@n3, B=@n4)) -> (ReturnValue))
  ]}
 TweakVehicleRulesetVisuals {nodes [
   n1 = TweakVehicleRulesetVisuals!() -> (then, Card:WBP Item Shop Tile Smart Card Object Reference, UsesVehicleRuleset:bool=~) {is=title K2Node_FunctionEntry}
   > SetCardSize!(
    execute
    self:WBP Item Shop Tile Smart Card Object Reference=@n1
    DesiredSize:int=GetSmartCardSize(UsesVehicleRuleset=@n1.UsesVehicleRuleset) -> (ReturnValue=0)
   ) -> (then)
   > SetCornerRadius!(
    execute
    self:WBP Item Shop Tile Smart Card Object Reference=Get Card() -> (Card:WBP Item Shop Tile Smart Card Object Reference)
    RadiusInPixels:float=SelectFloat(
     A=Get CornerRadiusVehicle() -> (CornerRadiusVehicle:double)
     B=Get CornerRadiusDefault() -> (CornerRadiusDefault:double)
     bPickA=Get UsesVehicleRuleset()
    )
   ) -> (then)
  ]}
 UpdateFilterListWidth {nodes [
   n1 = UpdateFilterListWidth!() -> (then, UpdatedWidth:double=~) {is=title K2Node_FunctionEntry}
   n2 = Get DynamicEntryBox_SmartCards_Vertical()
   n3 = float / float(
    A=@n1
    B:int=float + float(
     A:int=GetSmartCardSize(UsesVehicleRuleset=false) -> (ReturnValue=0)
     B=Get!(self:Dynamic Entry Box Base Object Reference=@n2) -> (EntrySpacing_X:double, EntrySpacing_Y:double)
    ) -> (ReturnValue:int)
   )
   n4 = Get FilterVM(self=Get FortItemShopVM())
   > Set MaxGridSizeOverride(MaxGridSizeOverride:Int Point Structure=Select!(
     Option 0_X:int=-1
     Option 0_Y:int=-1
     Option 1_X:int=Floor(A=@n3) {is=title K2Node_CallFunction}
     Option 1_Y:int=1
     Index=Get bIsActive(self=@n4)
    ) -> (ReturnValue:Int Point Structure)) -> (Output_Get:Int Point Structure)
   > n5 = Get!(execute:Exec) -> (then:Exec, else:Exec, FortItemShopOfferVM:Fort Item Shop Offer VM Object Reference)
   n6 = Get FilterVM(self=Get FortItemShopVM())
   > PopulateList(ItemsToShow=GetItemsToShow(
     GrantedItems=Get GrantedItems(self=@n5)
     SecondaryGrantedItems=Get DisplayedSecondaryGrantedItems(self=@n5)
     IsFilterActive=Get bIsActive(self=@n6)
     TileSize=Get TileSize(self=@n5) -> (TileSize=Size_1_x_1)
    ) {is=title K2Node_CallFunction}) {is=title K2Node_CallFunction}
  ]}
]
variables [
 CornerRadiusDefault {type=number, default=20}
 CornerRadiusVehicle {type=number, default=12}
 EntryBoxWrapOverride {type=EntryBoxWrapModeOverride, default=None}
 EntrySpacingDefault {type=Vector2D, default=<7 7>}
 EntrySpacingVehicle {type=Vector2D, default=<4 4>}
 IsCollapsedMode {type=boolean}
 IsUsedInBillboard {type=boolean}
 MaxGridSizeOverride {type=IntPoint, default=<-1 -1>}
 NewVar {type=string, default=Mary Jane Watson}
 SectionID {type=string, default=""}
 SmartCardSize {type=integer, default=94}
 SmartCardSize_Vehicle {type=integer, default=60}
 SmartCardSizeMobile {type=integer, default=80}
 SmartCardSizeMobile_Vehicle {type=integer, default=77}
]
widgets {
 info {inheritedWidgetCount=0, parentClass=&UserWidget, rootWidgetClass=&Overlay, widgetCount=8}
 widgets [
  Overlay_All {widgetClassPath=&Overlay, bIsVariable=true, widget=Overlay_All}
  WBP_CaptureForPostBufferUpdate {
   widgetClassPath=&WBP_CaptureForPostBufferUpdate
   bIsVariable=true
   parent=Overlay_All
   slot {class=OverlaySlot, properties {horizontalAlignment=HAlign_Fill, verticalAlignment=VAlign_Fill}}
  }
  Switcher_Mode {
   widgetClassPath=&CommonVisibilitySwitcher
   bIsVariable=true
   parent=Overlay_All
   widget=Switcher_Mode
   slot {class=OverlaySlot, properties {verticalAlignment=VAlign_Fill}}
  }
  SizeBox_All {
   widgetClassPath=&SizeBox
   bIsVariable=true
   parent=Switcher_Mode
   widget=SizeBox_All
   properties {bOverride_HeightOverride=true, bOverride_WidthOverride=true, heightOverride=195, widthOverride=195}
   slot {class=CommonVisibilitySwitcherSlot, properties {verticalAlignment=VAlign_Top}}
  }
  Switcher_EntryBoxWrapMode {
   widgetClassPath=&CommonVisibilitySwitcher
   bIsVariable=true
   parent=SizeBox_All
   widget=Switcher_EntryBoxWrapMode
   slot {class=SizeBoxSlot}
  }
  DynamicEntryBox_SmartCards_Vertical {
   widgetClassPath=/Script/FortniteUI.FortDynamicEntryBox
   bIsVariable=true
   parent=Switcher_EntryBoxWrapMode
   widget=DynamicEntryBox_SmartCards_Vertical
   properties {
    entryBoxType=VerticalWrap
    entryHorizontalAlignment=HAlign_Left
    entryVerticalAlignment=VAlign_Top
    entryWidgetClass=&/WBP_ItemShopTile_SmartCard.WBP_ItemShopTile_SmartCard_C
    maxElementSize=96
    numDesignerPreviewEntries=7
    entrySpacing=<7 7>
   }
   slot {class=CommonVisibilitySwitcherSlot}
  }
  DynamicEntryBox_SmartCards_Horizontal {
   widgetClassPath=&FortDynamicEntryBox
   bIsVariable=true
   parent=Switcher_EntryBoxWrapMode
   widget=DynamicEntryBox_SmartCards_Horizontal
   properties {
    entryBoxType=Wrap
    entryHorizontalAlignment=HAlign_Left
    entryVerticalAlignment=VAlign_Top
    entryWidgetClass=&WBP_ItemShopTile_SmartCard_C
    maxElementSize=96
    numDesignerPreviewEntries=7
    entrySpacing=<7 7>
   }
   slot {class=CommonVisibilitySwitcherSlot}
  }
  WBP_UIKit_Tag {
   widgetClassPath=/FortUILibrary/UIKit/Tag/WBP_UIKit_Tag
   bIsVariable=true
   parent=Switcher_Mode
   widget=WBP_UIKit_Tag
   properties {labelText="", tagText="[PH] Items"}
   slot {class=CommonVisibilitySwitcherSlot, properties {verticalAlignment=VAlign_Bottom}}
  }
 ]
}
