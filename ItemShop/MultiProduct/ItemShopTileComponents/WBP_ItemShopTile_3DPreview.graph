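# Widget-less item shop tile component that feeds an offer's 3D preview to the shop selection view model.
# SetCosmeticItemsInOffer builds preview data from Athena cosmetic items displayable in the current shop context.
# Hover or Request3DPreview publishes those items, or non-empty offer override previews, if row state and preview-enabled checks pass.
# Preview cycling is enabled only for hover-driven previews with multiple items, and unhovering disables it.
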WidgetBlueprint /ItemShop/MultiProduct/ItemShopTileComponents/WBP_ItemShopTile_3DPreview
graphs [
 EventGraph {nodes [
   n1 = Event SetOfferVM!() -> (OutputDelegate, then, OfferVM:Fort Item Shop Offer VM Object Reference) {is=title K2Node_Event}
   > SetFortItemShopOfferVM!(execute, self, Viewmodel:Fort Item Shop Offer VM Object Reference=@n1.OfferVM) -> (then)
   Event Request3DPreview() {is=title K2Node_Event}
   > n2 = Set CanCycle(CanCycle:bool=false) -> (Output_Get:bool)
   n3 = Get FortItemShopVM() -> (FortItemShopVM:Fort Item Shop VM Object Reference)
   n4 = Get FortItemShopSelectionVM() -> (FortItemShopSelectionVM:Fort Item Shop Selection VM Object Reference)
   n5 = Get RowState() -> (RowState:E_ItemShop_RowState Enum=NewEnumerator4)
   Branch(
    execute=[@n13.then, @n2.then]
    Condition=AND Boolean!(
     A=Select!(
      NewEnumerator4:bool=false
      NewEnumerator1:bool=false
      NewEnumerator0:bool=true
      NewEnumerator2:bool=true
      NewEnumerator3:bool=false
      Index:E_ItemShop_RowState Enum=@n5
     ) -> (ReturnValue:bool)
     B=IsItemPreviewEnabled!(self:Fort Item Shop VM Object Reference=@n3) -> (ReturnValue:bool)
     C:bool=IsItemPreviewEnabledForSelectedOfferGroup!(SelectionVM:Fort Item Shop Selection VM Object Reference=@n4) -> (IsEnabled:bool)
    ) -> (ReturnValue)
   )
   n6 = Get FortItemShopOfferVM() -> (FortItemShopOfferVM:Fort Item Shop Offer VM Object Reference)
   n7 = Get OverridePreviews!(self:Fort Item Shop Offer VM Object Reference=@n6) -> (OverridePreviews:Array of Item Shop Preview Data Object References)
   n8 = Is Not Empty!(TargetArray:Array of Item Shop Preview Data Object References=@n7) -> (ReturnValue:bool)
   n9 = IsValid(Object=@n6)
   > n10 = Set ItemsToPreview!(
    execute
    ItemsToPreview:Array of Item Shop Preview Data Object References=Select!(
     Option 0:Array of Item Shop Preview Data Object References=Get CosmeticItemsInOffer() -> (CosmeticItemsInOffer:Array of Item Shop Preview Data Object References)
     Option 1:Array of Item Shop Preview Data Object References=@n7
     Index:bool=AND Boolean!(A=@n9, B=@n8) -> (ReturnValue)
    ) -> (ReturnValue:Array of Item Shop Preview Data Object References)
    self:Fort Item Shop Selection VM Object Reference=Get FortItemShopSelectionVM()
   ) -> (then, Output_Get:Array of Item Shop Preview Data Object References)
   n11 = Length(TargetArray:Array of Item Shop Preview Data Object References=@n10)
   > Set bIsItemPreviewCycleEnabled!(
    execute
    bIsItemPreviewCycleEnabled:bool=AND Boolean(A="integer > integer"(A=@n11, B=1), B=Get CanCycle() -> (CanCycle:bool))
    self:Fort Item Shop Selection VM Object Reference=Get FortItemShopSelectionVM()
   ) -> (then, Output_Get:bool)
   n12 = Event SetHoveredState!() -> (OutputDelegate, then, IsHovered:bool, IsImmediate:bool) {is=title K2Node_Event}
   > Set CanCycle(CanCycle=true)
   > n13 = Branch(Condition=@n12.IsHovered)
   Set bIsItemPreviewCycleEnabled(execute=@n13.else, bIsItemPreviewCycleEnabled=false, self=Get FortItemShopSelectionVM())
   n14 = Event SetRowState!() -> (OutputDelegate, then, InRowState:E_ItemShop_RowState Enum=NewEnumerator4) {is=title K2Node_Event}
   > Set RowState(RowState:E_ItemShop_RowState Enum=@n14.InRowState) -> (Output_Get:E_ItemShop_RowState Enum=NewEnumerator4)
  ]}
 GetOfferVM {nodes [
   GetOfferVM() {is=title K2Node_FunctionEntry}
   > Return Node!(execute, OfferVM:Fort Item Shop Offer VM Object Reference)
  ]}
 GetTileAnchor {nodes [
   GetTileAnchor() {is=title K2Node_FunctionEntry}
   > Return Node!(execute, Anchor:Vector 2D Structure)
  ]}
 GetTileWidget {nodes [
   GetTileWidget() {is=title K2Node_FunctionEntry}
   > Return Node!(execute, NewParam:User Widget Object Reference)
  ]}
 SetCosmeticItemsInOffer {nodes [
   n1 = SetCosmeticItemsInOffer!() -> (then, InputPin:Array of Fort Item VM Object References) {is=title K2Node_FunctionEntry}
   > Clear(TargetArray:Array of Item Shop Preview Data Object References=Get CosmeticItemsInOffer())
   > Set OfferHasContentForCurrentContext(OfferHasContentForCurrentContext:bool=IsThereAnyItemForCurrentItemShopContext!(self:Fort Item Shop Offer VM Object Reference=Get FortItemShopOfferVM(), Items:Array of Fort Item VM Object References=@n1) -> (ReturnValue:bool)) -> (Output_Get:bool)
   > n2 = For Each Loop(Array:Array of Fort Item VM Object References=@n1) -> (Array Element:Fort Item VM Object Reference)
   > Branch(Condition=IsModeDisplayableInCurrentItemShopContext!(
     self:Fort Item Shop Offer VM Object Reference=Get FortItemShopOfferVM()
     ProductTag:Gameplay Tag Structure=GetCurrentProduct!(self:Fort Item VM Object Reference=@n2) -> (ReturnValue:Gameplay Tag Structure)
     bOfferHasContentForCurrentContext:bool=Get OfferHasContentForCurrentContext() -> (OfferHasContentForCurrentContext:bool)
    ) -> (ReturnValue:bool))
   > n3 = Cast To AthenaItemVM!(execute, Object=@n2) -> (then, CastFailed, AsAthena Item VM:Athena Item VM Object Reference) {is=title K2Node_DynamicCast}
   > Cast To AthenaCosmeticItemDefinition!(execute, Object=Get ItemDefinition!(self:Fort Item VM Object Reference=@n3) -> (ItemDefinition:Fort Item Definition Object Reference)) -> (then, CastFailed, AsAthena Cosmetic Item Definition:Athena Cosmetic Item Definition Object Reference) {is=title K2Node_DynamicCast}
   > n4 = CreatePreviewDataFromItemVM!(execute, Context:Widget Object Reference, ItemVM:Fort Item VM Object Reference=@n3) -> (then, ReturnValue:Item Shop Preview Data Object Reference)
   > Add(
    TargetArray:Array of Item Shop Preview Data Object References=Get CosmeticItemsInOffer()
    NewItem:"Item Shop Preview Data Object Reference (by ref)"=@n4
   ) {is=title K2Node_CallArrayFunction}
  ]}
]
variables [
 CanCycle {type=boolean}
 CosmeticItemsInOffer {type="array<ItemShopPreviewData>", default []}
 IsActive {type=boolean}
 RowState {type=E_ItemShop_RowState, default=None}
]
widgets {info {inheritedWidgetCount=0, parentClass=&UserWidget, rootWidgetClass=null, widgetCount=0}, widgets []}
