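# Item shop tile highlight widget, built on the UIKit highlight base, that animates in and out with the row state.
# Row state changes play Reveal or Hide through a stop-and-play macro, gated by IsActive, with a whoosh sound on both.
# SetOfferVM stores the offer and resets to invisible, while OnOfferSet stores it and reveals immediately.
# Helpers map violator intensity to highlight type, apply the background material and cap width.
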WidgetBlueprint /ItemShop/MultiProduct/ItemShopTileComponents/WBP_ItemShopTile_Highlight_2
animations [
 Reveal {
  binding_count=1
  length_seconds=0.25
  object=&self:Reveal
  sequence {bindings [WBP_ItemShopTile_Highlight_New_C_92 {
     id=#1
     possessed_class=WBP_ItemShopTile_Highlight_2_C
     tracks [
      {
       type=MovieSceneFloatTrack
       display_name=Render Opacity
       sections [{type=MovieSceneFloatSection, end_frame=5, start_frame=0, channels [None {type=MovieSceneScriptingFloatChannel, keys [[0,0],[5,1]]}]}]
      }
      {
       type=MovieScene2DTransformTrack
       sections [{
         type=MovieScene2DTransformSection
         end_frame=5
         start_frame=0
         channels [
          Translation.X {type=MovieSceneScriptingFloatChannel, keys [[0,150],[5,0]]}
          Scale.X {type=MovieSceneScriptingFloatChannel, keys [[0,1.25],[5,1]]}
          Scale.Y {type=MovieSceneScriptingFloatChannel, keys [[0,0.75],[5,1]]}
          Shear.X {type=MovieSceneScriptingFloatChannel, keys [[0,-16],[5,0]]}
         ]
        }]
      }
     ]
    }]}
 }
 Hide {
  binding_count=1
  length_seconds=0.05
  object=&self:Hide
  sequence {bindings [WBP_ItemShopTile_Highlight_New_C_94 {
     id=#2
     possessed_class=WBP_ItemShopTile_Highlight_2_C
     tracks [{
       type=MovieSceneFloatTrack
       display_name=Render Opacity
       sections [{type=MovieSceneFloatSection, end_frame=1, start_frame=0, channels [None {type=MovieSceneScriptingFloatChannel, keys [[0,1],[1,0]]}]}]
      }]
    }]}
 }
]
graphs [
 EventGraph {nodes [
   n1 = Event SetOfferVM!() -> (OutputDelegate, then, OfferVM:Fort Item Shop Offer VM Object Reference) {is=title K2Node_Event}
   > Set FortItemShopOfferVM(FortItemShopOfferVM:Fort Item Shop Offer VM Object Reference=@n1.OfferVM) -> (Output_Get:Fort Item Shop Offer VM Object Reference)
   > Set IsActive(IsActive:bool=false) -> (Output_Get:bool)
   > StopAllAnimations(self:User Widget Object Reference)
   > SetRenderOpacity(InOpacity=0)
   n2 = Event SetRowState!() -> (OutputDelegate, then, InRowState:E_ItemShop_RowState Enum=NewEnumerator4) {is=title K2Node_Event}
   > n3 = Switch on E_ItemShop_RowState!(execute, Selection:E_ItemShop_RowState Enum=@n2.InRowState) -> (NewEnumerator4:Exec, NewEnumerator1:Exec, NewEnumerator0:Exec, NewEnumerator2:Exec, NewEnumerator3:Exec) {is=title K2Node_SwitchEnum}
   n4 = Get!(execute:Exec=[@n3.NewEnumerator0, @n3.NewEnumerator2]) -> (then:Exec, else:Exec, IsActive:bool)
   Stop and Play Animation!(
    execute:Exec=[@n4.else, @n8.then]
    AnimationToStop:Widget Animation Object Reference=Get Hide() -> (Hide:Widget Animation Object Reference)
    AnimationToPlay:Widget Animation Object Reference=Get Reveal() -> (Reveal:Widget Animation Object Reference)
   ) -> (then:Exec) {is=title K2Node_MacroInstance}
   > n5 = Set IsActive(IsActive=true)
   SpawnSound2D!(
    execute=[@n6.then, @n5.then]
    Sound:Sound Base Object Reference=Get AnimationSound() -> (AnimationSound:Sound Base Object Reference)
    VolumeMultiplier:float=1
    PitchMultiplier:float=1
    StartTime:float=0
    ConcurrencySettings:Sound Concurrency Object Reference
    bPersistAcrossLevelTransition:bool=false
    bAutoDestroy:bool=true
   ) -> (then, ReturnValue:Audio Component Object Reference)
   Get(execute=[@n3.NewEnumerator1, @n3.NewEnumerator3])
   > Stop and Play Animation(AnimationToStop=Get Reveal(), AnimationToPlay=Get Hide())
   > n6 = Set IsActive(IsActive=false)
   n7 = Event OnOfferSet!() -> (
    OutputDelegate
    then
    OfferVM:Fort Item Shop Offer VM Object Reference
    ParentOfferVM:Fort Item Shop Offer VM Object Reference
    OfferGroupVM:Fort Item Shop Offer Group VM Object Reference
   ) {is=title K2Node_Event}
   > n8 = Set FortItemShopOfferVM(FortItemShopOfferVM=@n7.OfferVM)
  ]}
 GetHighlightTypeFromViolatorIntensity {nodes [
   n1 = GetHighlightTypeFromViolatorIntensity!() -> (then, ViolatorIntensity:EViolatorIntensity Enum=Low) {is=title K2Node_FunctionEntry}
   > Return Node!(
    execute
    HighlightType:E_UI_HighlightType Enum=Select!(
     Low:E_UI_HighlightType Enum=NewEnumerator4
     Medium:E_UI_HighlightType Enum=NewEnumerator4
     High:E_UI_HighlightType Enum=NewEnumerator0
     Urgent:E_UI_HighlightType Enum=NewEnumerator3
     Index:EViolatorIntensity Enum=@n1
    ) -> (ReturnValue:E_UI_HighlightType Enum)
   )
  ]}
 GetOfferVM {nodes [
   GetOfferVM() {is=title K2Node_FunctionEntry}
   > Return Node!(execute, OfferVM:Fort Item Shop Offer VM Object Reference)
  ]}
 GetTileAnchor {nodes [
   GetTileAnchor() {is=title K2Node_FunctionEntry}
   > Return Node!(execute, Anchor:Vector 2D Structure)
  ]}
 GetTileWidget {nodes [
   GetTileWidget() {is=title K2Node_FunctionEntry}
   > Return Node!(execute, NewParam:User Widget Object Reference)
  ]}
 SetBackgroundStyle {nodes [
   SetBackgroundStyle!() -> (then, HighlightType:E_UI_HighlightType Enum=NewEnumerator3) {is=title K2Node_FunctionEntry}
   > SetBrushFromMaterial(
    self=Get Background() -> (Background:Image Object Reference)
    Material=/FortUILibrary/Systems/Highlight/MaterialInstances/MI_UIKit_HighlightBackground
   )
   > Parent: SetBackgroundStyle!(execute, HighlightType:E_UI_HighlightType Enum=Get HighlightType() -> (HighlightType:E_UI_HighlightType Enum=NewEnumerator3)) -> (then) {is=title K2Node_CallParentFunction}
  ]}
 SetWidth {nodes [
   n1 = SetWidth!() -> (then, NewWidthOverride:double) {is=title K2Node_FunctionEntry}
   > n2 = Set WidthOverride(WidthOverride:double=@n1) -> (Output_Get:double)
   > n3 = Branch(Condition="float > float"(A=@n2))
   n4 = Get SB_WidthLimit() -> (SB_WidthLimit:Size Box Object Reference)
   > SetMaxDesiredWidth!(execute, self:Size Box Object Reference=@n4, InMaxDesiredWidth:float=@n2) -> (then)
   ClearMaxDesiredWidth(execute=@n3.else, self:Size Box Object Reference=@n4)
  ]}
 StopAndPlayAnimation {nodes [
   n1 = Inputs!() -> (execute, AnimationToStop:Widget Animation Object Reference, AnimationToPlay:Widget Animation Object Reference)
   > n2 = Branch(Condition=IsAnimationPlaying(InAnimation=@n1))
   > n3 = StopAnimation(InAnimation=@n1)
   PlayAnimationForward(execute=[@n3.then, @n2.else], InAnimation=@n1.AnimationToPlay)
   > Outputs!(then:Exec)
  ]}
]
variables [
 AnimationSound {type=SoundBase, default=/Game/Sounds/Fort_UI_Sounds/FrontEnd/Whooshes/Toasts/UI_Whoosh_Tiny_Appear_Cue}
 IsActive {type=boolean}
]
widgets {
 info {
  inheritedWidgetCount=0
  parentClass=/FortUILibrary/Systems/Highlight/WBP_UIKit_Highlight_Base.WBP_UIKit_Highlight_Base_C
  rootWidgetClass=null
  widgetCount=0
 }
 widgets []
}
