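# Item shop marquee widget that auto-scrolls named-slot content horizontally inside a non-interactive ScrollBox when it is wider than the box.
# Construct calls ResetAutoScroll, which starts a looping timer after a start delay; Destruct clears it.
# Each AutoScroll tick advances the offset by ScrollSpeed.
# At the end, it stops the timer, waits, plays the Fade animation, rewinds to zero, fades back in and restarts the timer.
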WidgetBlueprint /ItemShop/MultiProduct/WBP_ItemShop_AutoScrollHorizontally
animations [Fade {
  binding_count=1
  length_seconds=0.25
  object=&self:Fade
  sequence {bindings [ScrollBox {
     id=#1
     possessed_class=ScrollBox
     tracks [{
       type=MovieSceneFloatTrack
       display_name=Render Opacity
       sections [{type=MovieSceneFloatSection, end_frame=5, start_frame=0, channels [None {type=MovieSceneScriptingFloatChannel, keys [[0,1],[5,0]]}]}]
      }]
    }]}
 }]
graphs [
 EventGraph {nodes [
   AutoScroll() {is=title K2Node_CustomEvent}
   > Branch(Condition=NeedToScroll!() -> (NeedToScroll:bool))
   n1 = Get ScrollBox() -> (ScrollBox:Scroll Box Object Reference)
   > SetScrollOffset!(
    execute
    self:Scroll Box Object Reference=Get ScrollBox()
    NewScrollOffset:float=float + float(A=GetScrollOffset!(self:Scroll Box Object Reference=@n1) -> (ReturnValue:float), B=Get ScrollSpeed() -> (ScrollSpeed:double)) {is=title K2Node_CommutativeAssociativeBinaryOperator}
   ) -> (then)
   n2 = Get ScrollBox()
   n3 = GetScrollOffset(self=@n2)
   n4 = GetScrollOffsetOfEnd!(self:Scroll Box Object Reference=@n2) -> (ReturnValue:float)
   > Branch(Condition=AND Boolean(A="float >= float"(A=@n3, B=@n4), B=NeedToScroll()))
   > Clear and Invalidate Timer by Handle(Handle=Get ScrollTimerHandle() -> (ScrollTimerHandle:Timer Handle Structure))
   > Delay(Duration=Get ScrollEndDelay() -> (ScrollEndDelay:double))
   > PlayAnimationForward(InAnimation=Get Fade() -> (Fade:Widget Animation Object Reference))
   > Delay(Duration=0.25)
   > SetScrollOffset(self=Get ScrollBox(), NewScrollOffset=0)
   > PlayAnimationReverse(InAnimation=Get Fade())
   > StartAutoScrollTimer()
   Event Destruct()
   > Clear and Invalidate Timer by Handle(Handle=Get ScrollTimerHandle())
   Event Construct()
   > ResetAutoScroll()
  ]}
 NeedToScroll {nodes [
   NeedToScroll() {is=title K2Node_FunctionEntry}
   n1 = GetLocalSize!(Geometry:"Geometry Structure (by ref)"=GetCachedGeometry!(self:Widget Object Reference=Get ScrollBox()) -> (ReturnValue:"Geometry Structure (by ref)")) -> (ReturnValue:Vector 2D Structure)
   n2 = GetDesiredSize!(self:Widget Object Reference=Get ScrollBoxContent_NamedSlot() -> (ScrollBoxContent_NamedSlot:Named Slot Object Reference)) -> (ReturnValue:Vector 2D Structure)
   > Return Node!(execute, NeedToScroll:bool="float > float"(A=BreakVector2D(InVec=@n2), B=BreakVector2D(InVec=@n1)))
  ]}
 ResetAutoScroll {nodes [
   ResetAutoScroll() {is=title K2Node_FunctionEntry}
   > Clear and Invalidate Timer by Handle(Handle=Get ScrollTimerHandle())
   n1 = Get ScrollBox()
   > SetScrollOffset(self=@n1, NewScrollOffset=0)
   > Set RenderOpacity!(execute, RenderOpacity:float=1, self:Widget Object Reference=@n1) -> (then, Output_Get:float)
   > StartAutoScrollTimer()
  ]}
 StartAutoScrollTimer {nodes [
   StartAutoScrollTimer() {is=title K2Node_FunctionEntry}
   > n1 = Set Timer by Event(Delegate=Create Event(), Time=0.05, bLooping=true, InitialStartDelay=Get ScrollStartDelay() -> (ScrollStartDelay:double))
   > Set ScrollTimerHandle(ScrollTimerHandle:Timer Handle Structure=@n1) -> (Output_Get:Timer Handle Structure)
  ]}
]
variables [
 CanScroll {type=boolean}
 ScrollEndDelay {type=number, default=1.5}
 ScrollSpeed {type=number, default=1.5}
 ScrollStartDelay {type=number, default=1.5}
 ScrollTimerHandle {type=TimerHandle}
]
widgets {
 info {inheritedWidgetCount=0, parentClass=&UserWidget, rootWidgetClass=&ScrollBox, widgetCount=2}
 widgets [
  ScrollBox {
   widgetClassPath=&ScrollBox
   bIsVariable=true
   widget=ScrollBox
   properties {
    bAllowRightClickDragScrolling=false
    bEnableTouchScrolling=false
    bIsEnabled=false
    consumeMouseWheel=Never
    orientation=Orient_Horizontal
    scrollBarVisibility=Collapsed
    visibility=SelfHitTestInvisible
    scrollbarPadding=<ltrb 0 0 0 0>
    widgetStyle {
     barThickness=9
     bottomShadowBrush {
      bIsDynamicallyLoaded=false
      drawAs=Box
      imageType=FullColor
      mirroring=NoMirror
      resourceName=../../../Engine/Content/Slate/Common/ScrollBoxShadowBottom.png
      resourceObject=null
      tiling=NoTile
      imageSize=<16 8>
      margin=<ltrb 0.5 0 0.5 1>
      outlineSettings {
       bUseBrushTransparency=false
       roundingType=HalfHeightRadius
       width=0
       color {colorUseRule=UseColor_Specified, specifiedColor=<rgba 0 0 0 0>}
       cornerRadii=<0 0 0 0>
      }
      tintColor {colorUseRule=UseColor_Specified, specifiedColor=<rgba 1 1 1 1>}
      uVRegion {bIsValid=false, max=<0 0>, min=<0 0>}
     }
     horizontalScrolledContentPadding=<ltrb 0 0 1 0>
     leftShadowBrush {
      bIsDynamicallyLoaded=false
      drawAs=NoDrawType
      imageType=FullColor
      mirroring=NoMirror
      resourceName=../../../Engine/Content/Slate/Common/ScrollBoxShadowLeft.png
      resourceObject=null
      tiling=NoTile
      imageSize=<8 16>
      margin=<ltrb 1 0.5 0 0.5>
      outlineSettings {
       bUseBrushTransparency=false
       roundingType=HalfHeightRadius
       width=0
       color {colorUseRule=UseColor_Specified, specifiedColor=<rgba 0 0 0 0>}
       cornerRadii=<0 0 0 0>
      }
      tintColor {colorUseRule=UseColor_Specified, specifiedColor=<rgba 1 1 1 1>}
      uVRegion {bIsValid=false, max=<0 0>, min=<0 0>}
     }
     rightShadowBrush {
      bIsDynamicallyLoaded=false
      drawAs=NoDrawType
      imageType=FullColor
      mirroring=NoMirror
      resourceName=../../../Engine/Content/Slate/Common/ScrollBoxShadowRight.png
      resourceObject=null
      tiling=NoTile
      imageSize=<8 16>
      margin=<ltrb 0 0.5 1 0.5>
      outlineSettings {
       bUseBrushTransparency=false
       roundingType=HalfHeightRadius
       width=0
       color {colorUseRule=UseColor_Specified, specifiedColor=<rgba 0 0 0 0>}
       cornerRadii=<0 0 0 0>
      }
      tintColor {colorUseRule=UseColor_Specified, specifiedColor=<rgba 1 1 1 1>}
      uVRegion {bIsValid=false, max=<0 0>, min=<0 0>}
     }
     topShadowBrush {
      bIsDynamicallyLoaded=false
      drawAs=Box
      imageType=FullColor
      mirroring=NoMirror
      resourceName=../../../Engine/Content/Slate/Common/ScrollBoxShadowTop.png
      resourceObject=null
      tiling=NoTile
      imageSize=<16 8>
      margin=<ltrb 0.5 1 0.5 0>
      outlineSettings {
       bUseBrushTransparency=false
       roundingType=HalfHeightRadius
       width=0
       color {colorUseRule=UseColor_Specified, specifiedColor=<rgba 0 0 0 0>}
       cornerRadii=<0 0 0 0>
      }
      tintColor {colorUseRule=UseColor_Specified, specifiedColor=<rgba 1 1 1 1>}
      uVRegion {bIsValid=false, max=<0 0>, min=<0 0>}
     }
     verticalScrolledContentPadding=<ltrb 0 0 0 1>
    }
   }
  }
  ScrollBoxContent_NamedSlot {widgetClassPath=&NamedSlot, bIsVariable=true, parent=ScrollBox, widget=ScrollBoxContent_NamedSlot, slot {class=ScrollBoxSlot}}
 ]
}
