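# Item shop tile price widget that shows an offer's price, owned label or subscription text, plus an auto-scrolling discount disclaimer.
# A visibility switcher hosts the three states, with helper functions choosing subscription for Crew and ProgressiveCosmetic offers, otherwise owned or price.
# Subscription state maps to status text and color through selects.
# SetOfferVM stores the offer view model, and PreConstruct checks IsUsingTileV2Layout to apply smaller FrameBreak fonts and icon size.
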WidgetBlueprint /ItemShop/MultiProduct/ItemShopTileComponents/WBP_ItemShopTile_Price
graphs [
 EventGraph {nodes [
   Event PreConstruct()
   > Branch(Condition=IsUsingTileV2Layout!() -> (ReturnValue:bool))
   n1 = Get WBP_Price_ItemShop() -> (WBP_Price_ItemShop:WBP Price Item Shop Object Reference)
   > Set IconSize!(
    execute
    IconSize:Vector 2D Structure=Get FrameBreakIconSize() -> (FrameBreakIconSize:Vector 2D Structure)
    self:WBP Price Item Shop Object Reference=@n1
   ) -> (then, Output_Get:Vector 2D Structure)
   > Set PriceFont!(
    execute
    PriceFont:Slate Font Info Structure=Get FrameBreakPriceFont() -> (FrameBreakPriceFont:Slate Font Info Structure)
    self:WBP Price Item Shop Object Reference=@n1
   ) -> (then, Output_Get:Slate Font Info Structure)
   > Set AdditionalPriceFont!(
    execute
    AdditionalPriceFont:Slate Font Info Structure=Get FrameBreakAdditionalPriceFont() -> (FrameBreakAdditionalPriceFont:Slate Font Info Structure)
    self:WBP Price Item Shop Object Reference=@n1
   ) -> (then, Output_Get:Slate Font Info Structure)
   > Set RegularPriceFont!(
    execute
    RegularPriceFont:Slate Font Info Structure=Get FrameBreakRegularPriceFont() -> (FrameBreakRegularPriceFont:Slate Font Info Structure)
    self:WBP Price Item Shop Object Reference=@n1
   ) -> (then, Output_Get:Slate Font Info Structure)
   > Set AdditionalRegularPriceFont!(
    execute
    AdditionalRegularPriceFont:Slate Font Info Structure=Get FrameBreakAdditionalRegularPriceFont() -> (FrameBreakAdditionalRegularPriceFont:Slate Font Info Structure)
    self:WBP Price Item Shop Object Reference=@n1
   ) -> (then, Output_Get:Slate Font Info Structure)
   > UpdateVisuals(self:WBP Price Item Shop Object Reference=@n1)
   > Set Font!(
    execute
    Font:Slate Font Info Structure=Get FrameBreakOwnedFont() -> (FrameBreakOwnedFont:Slate Font Info Structure)
    self:Text Object Reference=Get Text_Owned() -> (Text_Owned:Common Text Object Reference)
   ) -> (then, Output_Get:Slate Font Info Structure)
   > Set Font(
    Font=Get FrameBreakSubscriptionFont() -> (FrameBreakSubscriptionFont:Slate Font Info Structure)
    self=Get Text_Subscription() -> (Text_Subscription:Common Text Object Reference)
   )
   n2 = Event SetOfferVM!() -> (OutputDelegate, then, OfferVM:Fort Item Shop Offer VM Object Reference) {is=title K2Node_Event}
   > Set FortItemShopOfferVM(FortItemShopOfferVM:Fort Item Shop Offer VM Object Reference=@n2.OfferVM) -> (Output_Get:Fort Item Shop Offer VM Object Reference)
  ]}
 GetOfferVM {nodes [
   GetOfferVM() {is=title K2Node_FunctionEntry}
   > Return Node!(execute, OfferVM:Fort Item Shop Offer VM Object Reference)
  ]}
 GetSubscriptionInfoColor {nodes [
   n1 = GetSubscriptionInfoColor!() -> (then, SubscriptionState:EMcpSubscriptionState Enum=Inactive) {is=title K2Node_FunctionEntry}
   n2 = Get ActiveStatusColor() -> (ActiveStatusColor:Slate Color Structure)
   n3 = Get DefaultStatusColor() -> (DefaultStatusColor:Slate Color Structure)
   > Return Node!(
    execute
    ReturnValue:Slate Color Structure=Select!(
     Inactive:Slate Color Structure=@n3
     Active:Slate Color Structure=@n2
     ActiveWithExternalBilling:Slate Color Structure=@n2
     NotAutoRenewable:Slate Color Structure=@n2
     Canceled:Slate Color Structure=@n2
     PaymentProcessError:Slate Color Structure=@n3
     PlatformAuthError:Slate Color Structure=@n3
     RewardsDelayed:Slate Color Structure=@n3
     Unavailable:Slate Color Structure=@n3
     Unknown:Slate Color Structure=@n3
     Index:EMcpSubscriptionState Enum=@n1
    ) -> (ReturnValue:Slate Color Structure)
   )
  ]}
 GetSubscriptionInfoText {nodes [
   n1 = GetSubscriptionInfoText!() -> (then, SubscriptionState:EMcpSubscriptionState Enum=Inactive) {is=title K2Node_FunctionEntry}
   n2 = Get DefaultStatusText() -> (DefaultStatusText:Text)
   n3 = Get ActiveStatusText() -> (ActiveStatusText:Text)
   > Return Node(ReturnValue:Text=Select(
     Inactive:Text=@n2
     Active:Text=@n3
     ActiveWithExternalBilling:Text=@n3
     NotAutoRenewable:Text=@n3
     Canceled:Text=@n3
     PaymentProcessError:Text=Get PaymentErrorText() -> (PaymentErrorText:Text)
     PlatformAuthError:Text=@n2
     RewardsDelayed:Text=@n2
     Unavailable:Text
     Unknown:Text=@n2
     Index=@n1
    ) -> (ReturnValue:Text))
  ]}
 GetSwitcherStateWidget {nodes [
   GetSwitcherStateWidget!() -> (then, IsOwned:bool=~, OfferType:EFortItemShopOfferType Enum=LevelBundle) {is=title K2Node_FunctionEntry}
   n1 = Get OfferType() -> (OfferType:EFortItemShopOfferType Enum=None)
   n2 = "Equal (Enum)"(A:EFortItemShopOfferType Enum=@n1, B:EFortItemShopOfferType Enum=ProgressiveCosmetic)
   n3 = "Equal (Enum)"(A=@n1, B=Crew)
   n4 = Get State_Price() -> (State_Price:Overlay Object Reference)
   n5 = Get State_Owned() -> (State_Owned:Overlay Object Reference)
   n6 = Get IsOwned() -> (IsOwned:bool)
   > Return Node(ReturnValue:Overlay Object Reference=Select!(
     Option 0:Overlay Object Reference=Select(Option 0=@n4, Option 1=@n5, Index=@n6)
     Option 1:Overlay Object Reference=Get State_Subscription() -> (State_Subscription:Overlay Object Reference)
     Index:bool=OR Boolean(A=@n3, B=@n2)
    ) -> (ReturnValue:Overlay Object Reference))
  ]}
 GetTileAnchor {nodes [
   GetTileAnchor() {is=title K2Node_FunctionEntry}
   > Return Node!(execute, Anchor:Vector 2D Structure)
  ]}
 GetTileWidget {nodes [
   GetTileWidget() {is=title K2Node_FunctionEntry}
   > Return Node!(execute, NewParam:User Widget Object Reference)
  ]}
]
variables [
 FrameBreakAdditionalPriceFont {
  type=SlateFontInfo
  default {
   bForceMonospaced=false
   bMaterialIsStencil=false
   fontMaterial=null
   fontObject=&HeadingNow
   letterSpacing=0
   monospacedWidth=1
   size=9
   skewAmount=0
   typefaceFontName=74 Regular
   outlineSettings {
    bApplyOutlineToDropShadows=false
    bMiteredCorners=false
    bSeparateFillAlpha=false
    outlineMaterial=null
    outlineSize=0
    outlineColor=<rgba 0 0 0 1>
   }
  }
 }
 FrameBreakAdditionalRegularPriceFont {
  type=SlateFontInfo
  default {
   bForceMonospaced=false
   bMaterialIsStencil=false
   fontMaterial=null
   fontObject=&HeadingNow
   letterSpacing=0
   monospacedWidth=1
   size=12.75
   skewAmount=0
   typefaceFontName=75 Medium
   outlineSettings {
    bApplyOutlineToDropShadows=false
    bMiteredCorners=false
    bSeparateFillAlpha=false
    outlineMaterial=null
    outlineSize=0
    outlineColor=<rgba 0 0 0 1>
   }
  }
 }
 FrameBreakIconSize {type=Vector2D, default=<20 20>}
 FrameBreakOwnedFont {
  type=SlateFontInfo
  default {
   bForceMonospaced=false
   bMaterialIsStencil=false
   fontMaterial=null
   fontObject=&HeadingNow
   letterSpacing=0
   monospacedWidth=1
   size=12.75
   skewAmount=0
   typefaceFontName=75 Medium
   outlineSettings {
    bApplyOutlineToDropShadows=false
    bMiteredCorners=false
    bSeparateFillAlpha=false
    outlineMaterial=null
    outlineSize=0
    outlineColor=<rgba 0 0 0 1>
   }
  }
 }
 FrameBreakPriceFont {
  type=SlateFontInfo
  default {
   bForceMonospaced=false
   bMaterialIsStencil=false
   fontMaterial=/FortUILibrary/Systems/ItemShopTile/Materials/PriceSwitcher/MI_UIKit_PriceSwitcher_Text
   fontObject=&HeadingNow
   letterSpacing=0
   monospacedWidth=1
   size=12.75
   skewAmount=0
   typefaceFontName=75 Medium
   outlineSettings {
    bApplyOutlineToDropShadows=false
    bMiteredCorners=false
    bSeparateFillAlpha=false
    outlineMaterial=null
    outlineSize=0
    outlineColor=<rgba 0 0 0 1>
   }
  }
 }
 FrameBreakRegularPriceFont {
  type=SlateFontInfo
  default {
   bForceMonospaced=false
   bMaterialIsStencil=false
   fontMaterial=&MI_UIKit_PriceSwitcher_Text
   fontObject=&HeadingNow
   letterSpacing=0
   monospacedWidth=1
   size=12.75
   skewAmount=0
   typefaceFontName=75 Medium
   outlineSettings {
    bApplyOutlineToDropShadows=false
    bMiteredCorners=false
    bSeparateFillAlpha=false
    outlineMaterial=null
    outlineSize=0
    outlineColor=<rgba 0 0 0 1>
   }
  }
 }
 FrameBreakSubscriptionFont {
  type=SlateFontInfo
  default {
   bForceMonospaced=false
   bMaterialIsStencil=false
   fontMaterial=null
   fontObject=&HeadingNow
   letterSpacing=0
   monospacedWidth=1
   size=12.75
   skewAmount=0
   typefaceFontName=75 Medium
   outlineSettings {
    bApplyOutlineToDropShadows=false
    bMiteredCorners=false
    bSeparateFillAlpha=false
    outlineMaterial=null
    outlineSize=0
    outlineColor=<rgba 0 0 0 1>
   }
  }
 }
 SubscriptionActiveStatusColor {type=SlateColor, default {colorUseRule=UseColor_Specified, specifiedColor=<rgba 0.0369 1 0.2232 1>}}
 SubscriptionDefaultStatusColor {type=SlateColor, default {colorUseRule=UseColor_Specified, specifiedColor=<rgba 1 1 1 1>}}
]
widgets {
 info {inheritedWidgetCount=0, namedSlotCount=2, parentClass=&UserWidget, rootWidgetClass=&VerticalBox, widgetCount=11}
 widgets [
  VerticalBox_0 {widgetClassPath=&VerticalBox, widget=VerticalBox_0}
  WBP_ItemShop_AutoScrollHorizontally {
   widgetClassPath=/ItemShop/MultiProduct/WBP_ItemShop_AutoScrollHorizontally
   parent=VerticalBox_0
   widget=WBP_ItemShop_AutoScrollHorizontally
   slot {class=VerticalBoxSlot}
  }
  Switcher_Ownership {
   widgetClassPath=&CommonVisibilitySwitcher
   bIsVariable=true
   namedSlotHost=WBP_ItemShop_AutoScrollHorizontally
   widget=Switcher_Ownership
  }
  State_Price {
   widgetClassPath=&Overlay
   bIsVariable=true
   parent=Switcher_Ownership
   widget=State_Price
   slot {class=CommonVisibilitySwitcherSlot}
  }
  WBP_Price_ItemShop {
   widgetClassPath=/ItemShop/MultiProduct/WBP_Price_ItemShop
   bIsVariable=true
   parent=State_Price
   widget=WBP_Price_ItemShop
   properties {
    discountDisclaimerText_Optional=&self:WidgetTree.DiscountDisclaimer
    additionalPricePadding=<ltrb 0 0 0 1>
    iconSize=<24 24>
    priceFont {
     bForceMonospaced=false
     bMaterialIsStencil=false
     fontMaterial=&MI_UIKit_PriceSwitcher_Text
     fontObject=&HeadingNow
     letterSpacing=0
     monospacedWidth=1
     size=15.75
     skewAmount=0
     typefaceFontName=75 Medium
     outlineSettings {
      bApplyOutlineToDropShadows=false
      bMiteredCorners=false
      bSeparateFillAlpha=false
      outlineMaterial=null
      outlineSize=0
      outlineColor=<rgba 0 0 0 1>
     }
    }
    regularPriceFont {
     bForceMonospaced=false
     bMaterialIsStencil=false
     fontMaterial=&MI_UIKit_PriceSwitcher_Text
     fontObject=&HeadingNow
     letterSpacing=0
     monospacedWidth=1
     size=15.75
     skewAmount=0
     typefaceFontName=75 Medium
     outlineSettings {
      bApplyOutlineToDropShadows=false
      bMiteredCorners=false
      bSeparateFillAlpha=false
      outlineMaterial=null
      outlineSize=0
      outlineColor=<rgba 0 0 0 1>
     }
    }
   }
   slot {class=OverlaySlot}
  }
  State_Owned {
   widgetClassPath=&Overlay
   bIsVariable=true
   parent=Switcher_Ownership
   widget=State_Owned
   slot {class=CommonVisibilitySwitcherSlot, properties {horizontalAlignment=HAlign_Left, verticalAlignment=VAlign_Center}}
  }
  Text_Owned {
   widgetClassPath=&CommonTextBlock
   bIsVariable=true
   parent=State_Owned
   widget=Text_Owned
   properties {
    justification=Center
    text=Owned
    textTransformPolicy=ToUpper
    colorAndOpacity {colorUseRule=UseColor_Specified, specifiedColor=<rgba 0.0369 1 0.2232 1>}
    font {
     bForceMonospaced=false
     bMaterialIsStencil=false
     fontMaterial=null
     fontObject=&HeadingNow
     letterSpacing=0
     monospacedWidth=1
     size=15.75
     skewAmount=0
     typefaceFontName=75 Medium
     outlineSettings {
      bApplyOutlineToDropShadows=false
      bMiteredCorners=false
      bSeparateFillAlpha=false
      outlineMaterial=null
      outlineSize=0
      outlineColor=<rgba 0 0 0 1>
     }
    }
    shadowOffset=<0 0>
   }
   slot {class=OverlaySlot}
  }
  State_Subscription {
   widgetClassPath=&Overlay
   bIsVariable=true
   parent=Switcher_Ownership
   widget=State_Subscription
   slot {class=CommonVisibilitySwitcherSlot}
  }
  Text_Subscription {
   widgetClassPath=&CommonTextBlock
   parent=State_Subscription
   widget=Text_Subscription
   properties {
    justification=Center
    mobileFontSizeMultiplier=1.1
    text=Learn More
    textTransformPolicy=ToUpper
    font {
     bForceMonospaced=false
     bMaterialIsStencil=false
     fontMaterial=null
     fontObject=&HeadingNow
     letterSpacing=0
     monospacedWidth=1
     size=15.75
     skewAmount=0
     typefaceFontName=75 Medium
     outlineSettings {
      bApplyOutlineToDropShadows=false
      bMiteredCorners=false
      bSeparateFillAlpha=false
      outlineMaterial=null
      outlineSize=0
      outlineColor=<rgba 0 0 0 1>
     }
    }
    shadowOffset=<0 0>
   }
   slot {class=OverlaySlot, properties {verticalAlignment=VAlign_Center}}
  }
  WBP_ItemShop_DiscountDisclaimer_AutoScrollHorizontally {
   widgetClassPath=&WBP_ItemShop_AutoScrollHorizontally
   bIsVariable=true
   parent=VerticalBox_0
   widget=WBP_ItemShop_DiscountDisclaimer_AutoScrollHorizontally
   slot {class=VerticalBoxSlot}
  }
  DiscountDisclaimer {
   widgetClassPath=&CommonTextBlock
   bIsVariable=true
   namedSlotHost=WBP_ItemShop_DiscountDisclaimer_AutoScrollHorizontally
   widget=DiscountDisclaimer
   properties {
    applyLineHeightToBottomLine=false
    bAutoCollapseWithEmptyText=true
    justification=Center
    mobileFontSizeMultiplier=1.2
    text="[PH]*Discount for select new players; full price applies to all others"
    font {
     bForceMonospaced=false
     bMaterialIsStencil=false
     fontMaterial=null
     fontObject=&HeadingNow
     letterSpacing=0
     monospacedWidth=1
     size=11.25
     skewAmount=0
     typefaceFontName=74 Regular
     outlineSettings {
      bApplyOutlineToDropShadows=false
      bMiteredCorners=false
      bSeparateFillAlpha=false
      outlineMaterial=null
      outlineSize=0
      outlineColor=<rgba 0 0 0 1>
     }
    }
    shadowOffset=<0 0>
   }
  }
 ]
}
