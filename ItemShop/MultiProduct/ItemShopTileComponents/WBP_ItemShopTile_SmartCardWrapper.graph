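# Item Shop tile wrapper that lazily hosts a Smart Card List widget in a named slot.
# Item array updates create or reuse the list and pass it the offer viewmodel, or remove it when empty.
# It forwards row state and section ID to the list and selects filtered secondary granted items when filtering is active.
# An active filter also overrides the grid to one row, with columns derived from width and card size.
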WidgetBlueprint /ItemShop/MultiProduct/ItemShopTileComponents/WBP_ItemShopTile_SmartCardWrapper
graphs [
 CreateOrGetInstance {nodes [
   CreateOrGetInstance() {is=title K2Node_FunctionEntry}
   > n1 = Get!(execute:Exec) -> (then:Exec, else:Exec, SmartCardListWidget:WBP Item Shop Tile Smart Card List Object Reference)
   > Return Node!(execute, Instance:WBP Item Shop Tile Smart Card List Object Reference=@n1)
   n2 = Create Widget!(
    execute:Exec=@n1.else
    Class:User Widget Class Reference=/ItemShop/MultiProduct/ItemShopTileComponents/SmartCardSystem/WBP_ItemShopTile_SmartCardList.WBP_ItemShopTile_SmartCardList_C
    OwningPlayer:Player Controller Object Reference=GetOwningPlayer()
    FortItemShopOfferVM:Fort Item Shop Offer VM Object Reference=Get FortItemShopOfferVM() -> (FortItemShopOfferVM:Fort Item Shop Offer VM Object Reference)
    MaxGridSizeOverride:Int Point Structure=Get GridSizeOverride() -> (GridSizeOverride:Int Point Structure)
    EntryBoxWrapOverride:EntryBoxWrapModeOverride Enum=Get EntryBoxWrapOverride() -> (EntryBoxWrapOverride:EntryBoxWrapModeOverride Enum=NewEnumerator2)
    IsUsedInBillboard:bool=Get IsUsedInBillboard() -> (IsUsedInBillboard:bool)
    SectionID:String=Get SectionID() -> (SectionID:String)
   ) -> (then:Exec, ReturnValue:WBP Item Shop Tile Smart Card List Object Reference)
   > Set SmartCardListWidget(SmartCardListWidget:WBP Item Shop Tile Smart Card List Object Reference=@n2) -> (Output_Get:WBP Item Shop Tile Smart Card List Object Reference)
   > AddChild(
    self=Get NamedSlot_SmartCardList() -> (NamedSlot_SmartCardList:Named Slot Object Reference)
    Content=Get SmartCardListWidget() -> (SmartCardListWidget:WBP Item Shop Tile Smart Card List Object Reference)
   )
   > Return Node(Instance=Get SmartCardListWidget())
  ]}
 EventGraph {nodes [
   n1 = Event SetRowState!() -> (OutputDelegate, then, InRowState:E_ItemShop_RowState Enum=NewEnumerator4) {is=title K2Node_Event}
   > n2 = Get()
   > SetRowState!(execute, self:BPI Item Shop Tile Interface=@n2, InRowState:E_ItemShop_RowState Enum=@n1.InRowState) -> (then)
   n3 = Event SetSectionID!() -> (OutputDelegate, then, InSectionID:String) {is=title K2Node_Event}
   > Set SectionID(SectionID:String=@n3.InSectionID) -> (Output_Get:String)
   > n4 = Get()
   > SetSectionID!(execute, self:BPI Item Shop Tile Interface=@n4, InSectionID:String=@n3.InSectionID) -> (then)
   n5 = Event SetOfferVM!() -> (OutputDelegate, then, OfferVM:Fort Item Shop Offer VM Object Reference) {is=title K2Node_Event}
   > SetFortItemShopOfferVM!(execute, self, Viewmodel:Fort Item Shop Offer VM Object Reference=@n5.OfferVM) -> (then)
  ]}
 GetItemsToShow {nodes [
   n1 = GetItemsToShow!() -> (
    then
    GrantedItems:Array of Fort Item VM Object References
    SecondaryGrantedItems:Array of Fort Item VM Object References
    IsFilterActive:bool=~
   ) {is=title K2Node_FunctionEntry}
   > n2 = GetSecondaryGrantedItemsForTile!(
    execute
    FilterVM:Fort Item Shop Filter VM Object Reference=Get FilterVM!(self:Fort Item Shop VM Object Reference=Get FortItemShopVM() -> (FortItemShopVM:Fort Item Shop VM Object Reference)) -> (FilterVM:Fort Item Shop Filter VM Object Reference)
    SectionId:Name=String To Name(InString=Get SectionID())
    InOfferSecondaryGrantedItems:Array of Fort Item VM Object References=@n1
   ) -> (then, ReturnValue:Array of Fort Item VM Object References)
   > Return Node!(
    execute
    ItemsToShow:Array of Fort Item VM Object References=Select(
     Option 0:Array of Fort Item VM Object References=@n1.SecondaryGrantedItems
     Option 1:Array of Fort Item VM Object References=@n2
     Index=@n1.IsFilterActive
    ) -> (ReturnValue:Array of Fort Item VM Object References)
   )
  ]}
 GetOfferVM {nodes [
   GetOfferVM() {is=title K2Node_FunctionEntry}
   > Return Node!(execute, OfferVM:Fort Item Shop Offer VM Object Reference)
  ]}
 GetTileAnchor {nodes [
   GetTileAnchor() {is=title K2Node_FunctionEntry}
   > Return Node!(execute, Anchor:Vector 2D Structure)
  ]}
 GetTileWidget {nodes [
   GetTileWidget() {is=title K2Node_FunctionEntry}
   > Return Node!(execute, NewParam:User Widget Object Reference)
  ]}
 UpdateGridSizeOverride {nodes [
   UpdateGridSizeOverride!() -> (then, UpdatedWidth:double=~, RefreshList:bool=~) {is=title K2Node_FunctionEntry}
   > Branch(Condition=IsUsingCompactUI())
   n1 = GetSmartCardSize!(self:WBP Item Shop Tile Smart Card List Object Reference=Get SmartCardListWidget(), UsesVehicleRuleset:bool=false) -> (ReturnValue:int=0)
   n2 = IsUsingCompactUI()
   n3 = float / float(
    A=Get UpdatedWidth() -> (UpdatedWidth:double)
    B:int=SelectInt(A=@n1, B=SelectInt(A=87, B=101, bPickA=@n2), bPickA="integer > integer"(A=@n1))
   )
   n4 = Get FilterVM(self=Get FortItemShopVM())
   > Set GridSizeOverride(GridSizeOverride:Int Point Structure=Select!(
     Option 0_X:int=-1
     Option 0_Y:int=-1
     Option 1_X:int=Floor(A=@n3) {is=title K2Node_CallFunction}
     Option 1_Y:int=1
     Index=Get bIsActive!(self:Fort Item Shop Filter VM Object Reference=@n4) -> (bIsActive:bool)
    ) -> (ReturnValue:Int Point Structure)) -> (Output_Get:Int Point Structure)
   > Branch(Condition=Get RefreshList() -> (RefreshList:bool))
   > UpdateFilterListWidth!(execute, self:WBP Item Shop Tile Smart Card List Object Reference=Get SmartCardListWidget(), UpdatedWidth:double=Get UpdatedWidth()) -> (then)
  ]}
 VM_ItemArrayUpdated {nodes [
   n1 = VM_ItemArrayUpdated!() -> (then, ItemsToDisplay:Array of Fort Item VM Object References) {is=title K2Node_FunctionEntry}
   > Get!(execute) -> (then, else, FortItemShopOfferVM:Fort Item Shop Offer VM Object Reference)
   > n2 = Branch(Condition="integer > integer"(A=Length(TargetArray:Array of Fort Item VM Object References=@n1)))
   > n3 = CreateOrGetInstance!(execute, self) -> (then, Instance:WBP Item Shop Tile Smart Card List Object Reference) {is=title K2Node_CallFunction}
   > SetOfferVM!(execute, self:BPI Item Shop Tile Interface=@n3, OfferVM:Fort Item Shop Offer VM Object Reference=Get FortItemShopOfferVM()) -> (then)
   n4 = Get!(execute=@n2.else) -> (then, else, SmartCardListWidget:WBP Item Shop Tile Smart Card List Object Reference)
   > RemoveChild(self=Get NamedSlot_SmartCardList(), Content=Get SmartCardListWidget())
   > Set SmartCardListWidget()
   Reroute Node(InputPin=@n4.else)
  ]}
]
variables [
 EntryBoxWrapOverride {type=EntryBoxWrapModeOverride, default=None}
 GridSizeOverride {type=IntPoint, default=<-1 -1>}
 IsUsedInBillboard {type=boolean}
 SectionID {type=string, default=""}
 SmartCardListWidget {type=WBP_ItemShopTile_SmartCardList_C, default=null}
]
widgets {
 info {inheritedWidgetCount=0, parentClass=&UserWidget, rootWidgetClass=&NamedSlot, widgetCount=1}
 widgets [NamedSlot_SmartCardList {widgetClassPath=&NamedSlot, bIsVariable=true, widget=NamedSlot_SmartCardList}]
}
