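# Item Shop tile image widget that loads an offer's presentations and swipes between them on a looping timer.
# Image_Current and Image_Next receive material, texture and ruleset framing, then PresentationSwap animates their swipe parameters before the index wraps onward.
# Row state pauses the timer and sets opacity for peek.
# It also loads the background, drives hover parameters, applies frame-break padding and shows a load guard for loading or broken images.
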WidgetBlueprint /ItemShop/MultiProduct/ItemShopTileComponents/WBP_ItemShopTile_SwipingOfferImage_Separate
animations [
 PresentationSwap {
  binding_count=2
  length_seconds=0.85
  object=&self:PresentationSwap
  sequence {bindings [
    Image_Current {
     id=#1
     possessed_class=Image
     tracks [{
       type=MovieSceneWidgetMaterialTrack
       sections [{
         type=MovieSceneParameterSection
         end_frame=17
         start_frame=0
         channels [
          Swipe {type=MovieSceneScriptingFloatChannel, keys [[3,0],[15,1]]}
          Swipe_Sharpness_V {type=MovieSceneScriptingFloatChannel, keys [[3,64],[9,32]]}
          FinalAlpha {type=MovieSceneScriptingFloatChannel, keys [[9,1],[15,0]]}
         ]
        }]
      }]
    }
    Image_Next {
     id=#2
     possessed_class=Image
     tracks [{
       type=MovieSceneWidgetMaterialTrack
       sections [{
         type=MovieSceneParameterSection
         end_frame=17
         start_frame=0
         channels [
          Swipe {type=MovieSceneScriptingFloatChannel, keys [[1,0],[13,1]]}
          Swipe_Sharpness_V {type=MovieSceneScriptingFloatChannel, keys [[1,32],[7,16]]}
          FinalAlpha {type=MovieSceneScriptingFloatChannel, keys [[0,0],[5,1]]}
          InvertSwipe {type=MovieSceneScriptingFloatChannel, const [1,1]}
         ]
        }]
      }]
    }
   ]}
 }
 Hover {
  binding_count=2
  length_seconds=2
  object=&self:Hover
  sequence {bindings [
    Image_Current {
     id=#3
     possessed_class=Image
     tracks [{
       type=MovieSceneWidgetMaterialTrack
       sections [{
         type=MovieSceneParameterSection
         end_frame=40
         start_frame=0
         channels [
          Hover {type=MovieSceneScriptingFloatChannel, keys [[0,0],[4,1]]}
          HoverBurst {type=MovieSceneScriptingFloatChannel, keys [[0,0],[2,3],[4,0],[40,1]]}
         ]
        }]
      }]
    }
    Image_Next {
     id=#4
     possessed_class=Image
     tracks [{
       type=MovieSceneWidgetMaterialTrack
       sections [{
         type=MovieSceneParameterSection
         end_frame=40
         start_frame=0
         channels [
          Hover {type=MovieSceneScriptingFloatChannel, keys [[0,0],[4,1]]}
          HoverBurst {type=MovieSceneScriptingFloatChannel, keys [[0,0],[2,3],[4,0],[40,1]]}
         ]
        }]
      }]
    }
   ]}
 }
 Unhover {
  binding_count=2
  length_seconds=0.0999
  object=&self:Unhover
  sequence {bindings [
    Image_Current {
     id=#3
     possessed_class=Image
     tracks [{
       type=MovieSceneWidgetMaterialTrack
       sections [{
         type=MovieSceneParameterSection
         end_frame=1
         start_frame=0
         channels [
          Hover {type=MovieSceneScriptingFloatChannel, keys [[0,1],[1.9973,0]]}
          HoverBurst {type=MovieSceneScriptingFloatChannel, const [0,2]}
         ]
        }]
      }]
    }
    Image_Next {
     id=#4
     possessed_class=Image
     tracks [{
       type=MovieSceneWidgetMaterialTrack
       sections [{
         type=MovieSceneParameterSection
         end_frame=1
         start_frame=0
         channels [
          Hover {type=MovieSceneScriptingFloatChannel, keys [[0,1],[1.9973,0]]}
          HoverBurst {type=MovieSceneScriptingFloatChannel, const [0,2]}
         ]
        }]
      }]
    }
   ]}
 }
 Peek {
  binding_count=1
  length_seconds=0.2
  object=&self:Peek
  sequence {
   bindings [WBP_ItemShopTile_SwipingOfferImage_Separate_C_10 {
     id=#5
     possessed_class=WBP_ItemShopTile_SwipingOfferImage_Separate_C
     tracks [{
       type=MovieSceneDoubleTrack
       display_name=Peek Anim Alpha
       sections [{type=MovieSceneDoubleSection, end_frame=12, start_frame=0, channels [None {type=MovieSceneScriptingDoubleChannel, keys [[0,0],[12,1]]}]}]
      }]
    }]
   display_rate {numerator=60}
  }
 }
]
graphs [
 ApplyRulesetToMID {nodes [
   ApplyRulesetToMID!() -> (then, MID:Material Instance Dynamic Object Reference, Rule:S Item Shop Image Ruleset Values Structure) {is=title K2Node_FunctionEntry}
   > n1 = Sequence!(execute) -> (then_0, then_1, then_2:Exec, then_3:Exec, then_4:Exec)
   n2 = Get!() -> (
    Rule_Position_6_#6:Vector 2D Structure
    Rule_Zoom_9_#7:double=~
    Rule_HasRoundedCorners_14_#8:bool=~
    Rule_ScaleToFitX_17_#9:bool=~
    Rule_Position_FrameBreaking_25_#10:Vector 2D Structure
    Rule_Zoom_FrameBreaking_23_#11:double=~
   )
   n3 = BreakVector2D(InVec=Select(
     Option 0:Vector 2D Structure=@n2
     Option 1:Vector 2D Structure=@n2.Rule_Position_FrameBreaking_25_#10
     Index=Get IsFrameBreaking() -> (IsFrameBreaking:bool)
    ) -> (ReturnValue:Vector 2D Structure))
   > SetScalarParameterValue(self=Get MID() -> (MID:Material Instance Dynamic Object Reference), ParameterName=X Anchor Point, Value=@n3)
   SetScalarParameterValue(execute=@n1.then_1, self=Get MID(), ParameterName=Y Anchor Point, Value=@n3.Y)
   n4 = Get() -> (
    Rule_Zoom_9_#7=~
    Rule_HasRoundedCorners_14_#8=~
    Rule_ScaleToFitX_17_#9=~
    Rule_Zoom_FrameBreaking_23_#11=~
   )
   SetScalarParameterValue(
    execute=@n1.then_2
    self=Get MID()
    ParameterName=Zoom Level
    Value=Select(
     Option 0:double=@n4.Rule_Zoom_9_#7
     Option 1:double=@n4.Rule_Zoom_FrameBreaking_23_#11
     Index=Get IsFrameBreaking()
    ) -> (ReturnValue:double)
   )
   n5 = Get() -> (
    Rule_Zoom_9_#7=~
    Rule_HasRoundedCorners_14_#8=~
    Rule_ScaleToFitX_17_#9=~
    Rule_Zoom_FrameBreaking_23_#11=~
   )
   SetScalarParameterValue(
    execute=@n1.then_3
    self=Get MID()
    ParameterName=bHasRoundedCorners
    Value="To Float (Boolean)"(InBool=@n5.Rule_HasRoundedCorners_14_#8)
   )
   SetScalarParameterValue(
    execute=@n1.then_4
    self=Get MID()
    ParameterName=ScaleToFitX
    Value="To Float (Boolean)"(InBool=@n5.Rule_ScaleToFitX_17_#9)
   )
  ]}
 CF_GetPresentationData {nodes [
   n1 = CF_GetPresentationData!() -> (
    then
    SecondaryGrantedItems:Array of Fort Item VM Object References
    IsFilterActive:bool=~
    TileSize:EItemShopTileSize Enum=Size_1_x_1
    OfferType:EFortItemShopOfferType Enum=LevelBundle
    UsingFullTileOverridePresentations:bool=~
    DisplayAssetPath:String
    TileDimensions:Vector 2D Structure
   ) {is=title K2Node_FunctionEntry}
   > Return Node!(execute, Out_DisplayAssetPath:String=@n1.DisplayAssetPath)
  ]}
 EventGraph {nodes [
   n1 = OnPresentationsLoaded!() -> (OutputDelegate, then, LoadedPresentations:Array of Fort Item Shop Loaded Presentation Structures) {is=title K2Node_CustomEvent}
   > K2Node_Composite_4 = debug logs - on presentations loaded!(execute, Array:Array of Fort Item Shop Loaded Presentation Structures=@n1.LoadedPresentations) -> (then:Exec) {is=title K2Node_Composite}
   > OnSetPresentations!(execute, self, Images:Array of Fort Item Shop Loaded Presentation Structures=@n1.LoadedPresentations) -> (then)
   OnSetBrokenImage() {is=title K2Node_CustomEvent}
   > SetPresentationsLoadingState!(execute, self, LoadState:E_ItemShop_LoadGuardState Enum=NewEnumerator2, ShouldClearPresentationsArray:bool=false) -> (then)
   n2 = OnPresentationSwap() {is=title K2Node_CustomEvent}
   n3 = AnimationFinished() {is=title K2Node_CustomEvent}
   n4 = Get PresentationSwap() -> (PresentationSwap:Widget Animation Object Reference)
   UnbindFromAnimationFinished!(execute=@n2.then, self:User Widget Object Reference, Animation:Widget Animation Object Reference=@n4, Delegate:Delegate=@n3) -> (then)
   > BindToAnimationFinished(Animation=@n4, Delegate=@n3)
   > PlayAnimationForward(InAnimation=Get PresentationSwap())
   n5 = DesignTimePreviewMaterial!() -> (OutputDelegate, then, MaterialToPreview:"Material Interface Object Reference (by ref)") {is=title K2Node_CustomEvent}
   > SetBrushFromMaterial(self=Get Image_Current() -> (Image_Current:Image Object Reference), Material=@n5.MaterialToPreview)
   n6 = OnSetPresentations!() -> (OutputDelegate, then, Images:Array of Fort Item Shop Loaded Presentation Structures) {is=title K2Node_CustomEvent}
   > Set LoadedPresentations(LoadedPresentations:Array of Fort Item Shop Loaded Presentation Structures=@n6.Images) -> (Output_Get:Array of Fort Item Shop Loaded Presentation Structures)
   > StartShowingPresentations!(execute, self, PresentationSwapDelegate:Delegate=@n2) -> (then)
   n7 = ToggleRulesetDebug!() -> (OutputDelegate, then, IsEnabled:bool=~) {is=title K2Node_CustomEvent}
   > Set Visibility(
    Visibility=Select(Option 0:ESlateVisibility Enum=Collapsed, Option 1:ESlateVisibility Enum=SelfHitTestInvisible, Index=@n7.IsEnabled) -> (ReturnValue:ESlateVisibility Enum)
    self=Get Debug_Ruleset() -> (Debug_Ruleset:Common Text Object Reference)
   )
   n8 = Event RequestLoadPresentations!() -> (OutputDelegate, then, DisplayAssetPath:String) {is=title K2Node_Event}
   > Set with Broadcast DisplayAssetPath_BP!(execute, DisplayAssetPath_BP:String=@n8.DisplayAssetPath) -> (then, Output_Get:String)
   n9 = Event SetSectionID!() -> (OutputDelegate, then, InSectionID:String) {is=title K2Node_Event}
   > Set SectionID(SectionID:String=@n9.InSectionID) -> (Output_Get:String)
   Event ResetTile() {is=title K2Node_Event}
   > ResetTile_Implementation()
   n10 = Event SetRowState!() -> (OutputDelegate, then, InRowState:E_ItemShop_RowState Enum=NewEnumerator4) {is=title K2Node_Event}
   > Set Row State Implementation!(execute, self, InRowState:E_ItemShop_RowState Enum=@n10.InRowState) -> (then) {is=title K2Node_CallFunction}
   n11 = Event SetPeekState!() -> (OutputDelegate, then, IsPeeked:bool) {is=title K2Node_Event}
   > SetPeekState_Implementation!(execute, self, InIsPeeked:bool=@n11.IsPeeked) -> (then)
   n12 = Event PreConstruct()
   > Branch(Condition=@n12.IsDesignTime)
   > SetPresentationsLoadingState(LoadState=NewEnumerator1, ShouldClearPresentationsArray=false)
   n13 = Get IsPreviewingFrameBreak() -> (IsPreviewingFrameBreak:bool)
   > Set with Broadcast IsFrameBreaking!(execute, IsFrameBreaking:bool=@n13) -> (then, Output_Get:bool)
   > SetFrameBreakStyle!(execute, self, IsFrameBreaking:bool=@n13) -> (then)
   n14 = Event SetOfferVM!() -> (OutputDelegate, then, OfferVM:Fort Item Shop Offer VM Object Reference) {is=title K2Node_Event}
   > SetFortItemShopOfferVM!(execute, self, Viewmodel:Fort Item Shop Offer VM Object Reference=@n14.OfferVM) -> (then)
   n15 = Event SetHoveredState!() -> (OutputDelegate, then, IsHovered:bool, IsImmediate:bool) {is=title K2Node_Event}
   > n16 = Set Is Hovered(Is Hovered:bool=@n15.IsHovered) -> (Output_Get:bool)
   > n17 = Branch(Condition=@n16)
   > StopAnimation(InAnimation=Get Unhover() -> (Unhover:Widget Animation Object Reference))
   > PlayAnimationForward(InAnimation=Get Hover() -> (Hover:Widget Animation Object Reference))
   StopAnimation(execute=@n17.else, InAnimation=Get Hover())
   > PlayAnimationForward(InAnimation=Get Unhover())
   Event InitializeTile() {is=title K2Node_Event}
   > InitializeTile_Implementation()
   n18 = Event SetTileDimensions!() -> (OutputDelegate, then, Dimensions:Vector 2D Structure) {is=title K2Node_Event}
   > Set with Broadcast TileDimensions!(execute, TileDimensions:Vector 2D Structure=@n18.Dimensions) -> (then, Output_Get:Vector 2D Structure)
   SetHoveredStateVisual!(execute=@n20.then, self, IsHovered:bool=Get Is Hovered() -> (Is Hovered:bool)) -> (then)
   n19 = GetNextPresentationIndex!() -> (ReturnValue:int=0)
   n20 = UpdatePresentations(execute=@n21.then)
   n21 = Set CurrentPresentationIndex(execute=@n3.then, CurrentPresentationIndex:int=@n19) -> (Output_Get:int=0)
  ]}
 debug logs - on presentations loaded {
  path=EventGraph.K2Node_Composite_4.debug logs - on presentations loaded
  nodes [
   n1 = Inputs!() -> (execute, Array:Array of Fort Item Shop Loaded Presentation Structures)
   > n2 = Branch(Condition=Get EnableLogs() -> (EnableLogs:bool))
   > n3 = Get!(execute:Exec) -> (then:Exec, else:Exec, FortItemShopOfferVM:Fort Item Shop Offer VM Object Reference)
   > n4 = For Each Loop!(Exec, Array:Array of Fort Item Shop Loaded Presentation Structures=@n1) -> (
    LoopBody
    Array Element_Texture:Texture 2D Object Reference
    Array Element_Material:Material Interface Object Reference
    Array Index
    Completed
   )
   n5 = GetTitle!(self:Fort Base Offer VM Object Reference=@n3) -> (ReturnValue:Text)
   n6 = Length(TargetArray:Array of Fort Item Shop Loaded Presentation Structures=@n1)
   n7 = GetOwningWidgetName!(self) -> (ReturnValue:String)
   > PrintString!(
    execute
    InString=Append!(
     A="Presentations loaded for Offer: "
     B="To String (Text)"(InText=@n5)
     C:String=" - num presentations: "
     D:String="To String (Integer)"(InInt=@n6)
     E:String=" - texture:"
     F:String=GetDisplayName(Object=@n4)
     G:String=" - Object: "
     H:String=@n7
    ) -> (ReturnValue)
    bPrintToScreen=false
    bPrintToLog
    TextColor
    Duration
    Key
   ) -> (then)
   Outputs!(then:Exec=[@n2.else, @n8.then, @n4.Completed])
   n8 = PrintString(
    execute=@n3.else
    InString=Append!(A="Presentations loaded for offer: invalid Offer VM - Object: ", B=@n7) -> (ReturnValue)
    bPrintToScreen=false
   )
  ]
 }
 GetNextPresentationIndex {nodes [
   GetNextPresentationIndex() {is=title K2Node_FunctionEntry}
   n1 = Get CurrentPresentationIndex() -> (CurrentPresentationIndex:int=0)
   n2 = Get PresentationsAmount() -> (PresentationsAmount:int=0)
   > Return Node!(execute, ReturnValue:int="Wrap (Integer)"!(Value:int=int + int(A=@n1), Min:int=-1, Max:int=int - int(A=@n2)) -> (ReturnValue:int=0))
  ]}
 GetOfferVM {nodes [
   GetOfferVM() {is=title K2Node_FunctionEntry}
   > Return Node!(
    execute
    OfferVM:Fort Item Shop Offer VM Object Reference=Get FortItemShopOfferVM() -> (FortItemShopOfferVM:Fort Item Shop Offer VM Object Reference)
   )
  ]}
 GetOwningWidgetName {nodes [
   GetOwningWidgetName() {is=title K2Node_FunctionEntry}
   n1 = GetOuterObject!(Object:Object Reference=GetParent!(self:Widget Object Reference) -> (ReturnValue:Panel Widget Object Reference)) -> (ReturnValue:Object Reference)
   > Return Node!(execute, ReturnValue:String=GetDisplayName(Object=GetOuterObject(Object=@n1)))
  ]}
 GetTileAnchor {nodes [
   GetTileAnchor() {is=title K2Node_FunctionEntry}
   > Return Node!(execute, Anchor:Vector 2D Structure)
  ]}
 GetTileWidget {nodes [
   GetTileWidget() {is=title K2Node_FunctionEntry}
   > Return Node!(execute, NewParam:User Widget Object Reference)
  ]}
 InitializeTile_Implementation {nodes [
   InitializeTile_Implementation() {is=title K2Node_FunctionEntry}
   > Clear and Invalidate Timer by Handle(Handle=Get PresentationSwapTimerHandle() -> (PresentationSwapTimerHandle:Timer Handle Structure))
   > Set CurrentPresentationIndex(CurrentPresentationIndex=0) -> (Output_Get=0)
   > UnbindAllFromAnimationFinished!(execute, self:User Widget Object Reference, Animation:Widget Animation Object Reference=Get PresentationSwap()) -> (then)
   > StopAnimation(InAnimation=Get PresentationSwap())
   > SetPresentationsLoadingState(LoadState=NewEnumerator0, ShouldClearPresentationsArray=true)
  ]}
 LoadPresentations {nodes [
   LoadPresentations() {is=title K2Node_FunctionEntry}
   > SetPresentationsLoadingState(LoadState=NewEnumerator0, ShouldClearPresentationsArray=true)
   > K2Node_Composite_0 = debug logs - Load presentations!(execute) -> (then:Exec) {is=title K2Node_Composite}
   n1 = Get FortItemShopOfferVM()
   n2 = Get FortItemShopOfferVM()
   n3 = Break S Item Shop Image Ruleset Values!(S_ItemShop_ImageRulesetValues:"S Item Shop Image Ruleset Values Structure (by ref)"=Get ImageDisplayRuleset() -> (ImageDisplayRuleset:S Item Shop Image Ruleset Values Structure)) -> (
    Zoom_9_#7:double=~
    ScaleToFitX_17_#9:bool=~
    Zoom_FrameBreaking_23_#11:double=~
   ) {is=title K2Node_BreakStruct}
   n4 = Get IsFrameBreaking()
   n5 = Get!() -> (TileDimensions_X:double, TileDimensions_Y:double)
   n6 = Round(A=float * float(
     A=SelectFloat(A=@n3.Zoom_FrameBreaking_23_#11, B=@n3, bPickA=@n4)
     B=Select(Option 0:double=@n5.TileDimensions_Y, Option 1:double=@n5, Index=@n3.ScaleToFitX_17_#9) -> (ReturnValue:double)
    ) {is=title K2Node_PromotableOperator}) {is=title K2Node_CallFunction}
   > LoadPresentations!(
    execute
    self:Fort Item Shop Offer VM Object Reference=@n2
    Owner:Object Reference
    InDisplayAssetPath:String=Get DisplayAssetPath_BP() -> (DisplayAssetPath_BP:String)
    DefaultMaterial:Material Interface Object Reference=Get DefaultPresentationMaterial() -> (DefaultPresentationMaterial:Material Interface Object Reference)
    bAllowFullTileOverridePresentations:bool="Equal (Enum)"(
     A:EItemShopTileSize Enum=Get TileSize!(self:Fort Item Shop Offer VM Object Reference=@n1) -> (TileSize:EItemShopTileSize Enum=Size_1_x_1)
     B:EItemShopTileSize Enum=Size_4_x_1
    )
    TextureSize:int=@n6
    OnLoadedCallback:"Delegate (by ref)"=Create Event()
    OnLoadFailedCallback:"Delegate (by ref)"=Create Event()
   ) -> (then) {is=title K2Node_CallFunction}
   K2Node_Composite_2 = Debug Logs: Texture Size And Zoom Values!(
    execute
    ShopOfferVM:Fort Base Offer VM Object Reference=@n2
    Zoom:double=@n3
    ZoomFrameBreaking:double=@n3.Zoom_FrameBreaking_23_#11
    TextureSize:int=@n6
   ) {is=title K2Node_Composite}
  ]}
 debug logs - Load presentations {
  path=LoadPresentations.K2Node_Composite_0.debug logs - Load presentations
  nodes [
   Inputs!() -> (execute)
   > n1 = Branch(Condition=Get EnableLogs())
   > n2 = Get!(execute:Exec) -> (then:Exec, else:Exec, FortItemShopOfferVM:Fort Item Shop Offer VM Object Reference)
   n3 = GetTitle(self=@n2)
   n4 = GetOwningWidgetName() {is=title K2Node_CallFunction}
   > n5 = PrintString(
    InString=Append!(A="Tile - Load Presentations for VM: ", B="To String (Text)"(InText=@n3), C:String=" - Object: ", D:String=@n4) -> (ReturnValue)
    bPrintToScreen=false
   )
   Outputs(then=[@n6.then, @n5.then, @n1.else])
   n6 = PrintString(execute=@n2.else, InString=Append!(A="Tile - Load Presentations - invalid VM - Object: ", B=@n4) -> (ReturnValue), bPrintToScreen=false)
  ]
 }
 "Debug Logs: Texture Size And Zoom Values" {
  path=LoadPresentations.K2Node_Composite_2.Debug Logs: Texture Size And Zoom Values
  nodes [
   n1 = Inputs!() -> (execute, ShopOfferVM:Fort Base Offer VM Object Reference, Zoom:double=~, ZoomFrameBreaking:double=~, TextureSize:int)
   n2 = GetTitle(self=@n1)
   n3 = Get IsFrameBreaking()
   > PrintString(InString=Append!(
     A="Offer: "
     B="To String (Text)"(InText=@n2)
     C:String=" zoom:zoomfb: "
     D:String="To String (Float)"(InDouble=@n1.Zoom)
     E:String=:
     F:String="To String (Float)"(InDouble=@n1.ZoomFrameBreaking)
     G:String=" isFrameBreaking = "
     H:String="To String (Boolean)"(InBool=@n3)
     I:String=" textureSize = "
     J:String="To String (Integer)"(InInt=@n1.TextureSize)
    ) -> (ReturnValue))
   Outputs!()
  ]
 }
 OnBackgroundOverrideLoaded {nodes [
   n1 = OnBackgroundOverrideLoaded!() -> (then, LoadedBackground:Object Reference) {is=title K2Node_FunctionEntry}
   > n2 = Cast To MaterialInterface!(execute, Object=@n1) -> (then, CastFailed, AsMaterial Interface:Material Interface Object Reference) {is=title K2Node_DynamicCast}
   > SetBackgroundMaterial!(execute, self, Material:Material Interface Object Reference=@n2) -> (then)
   n3 = Cast To Texture2D(execute=@n2.CastFailed, Object=@n1)
   > SetBrushFromMaterial(
    self=Get Background() -> (Background:Image Object Reference)
    Material=Get DefaultBackgroundTextureMaterial() -> (DefaultBackgroundTextureMaterial:Material Interface Object Reference)
   )
   > n4 = GetDynamicMaterial(self=Get Background())
   > SetTextureParameterValue(self=@n4, ParameterName=Texture, Value=@n3)
   > SetBackgroundLoadingState!(execute, self, IsLoading:bool=false) -> (then)
   SetBackgroundMaterial(
    execute=@n3.CastFailed
    Material=Get DefaultBackgroundGradientMaterial() -> (DefaultBackgroundGradientMaterial:Material Interface Object Reference)
   )
  ]}
 Reset Image Material Params {nodes [
   Reset Image Material Params() {is=title K2Node_FunctionEntry}
   > n1 = Sequence!(execute) -> (then_0, then_1)
   > n2 = GetDynamicMaterial(self=Get Image_Current())
   > SetScalarParameterValue(self=@n2, ParameterName=Swipe, Value=0)
   > SetScalarParameterValue(self=@n2, ParameterName=Swipe_Sharpness_V, Value=32)
   > SetScalarParameterValue(self=@n2, ParameterName=FinalAlpha, Value=1)
   n3 = GetDynamicMaterial(execute=@n1.then_1, self=Get Image_Next() -> (Image_Next:Image Object Reference))
   > SetScalarParameterValue(self=@n3, ParameterName=Swipe, Value=0)
   > SetScalarParameterValue(self=@n3, ParameterName=Swipe_Sharpness_V, Value=32)
   > SetScalarParameterValue(self=@n3, ParameterName=FinalAlpha, Value=0)
  ]}
 ResetTile_Implementation {nodes [
   ResetTile_Implementation() {is=title K2Node_FunctionEntry}
   n1 = Get FortItemShopOfferVM()
   > CancelLoadingBackground!(execute, self:Fort Item Shop Offer VM Object Reference=@n1, Owner:Object Reference) -> (then)
   > CancelLoadingPresentations!(execute, self:Fort Item Shop Offer VM Object Reference=@n1, Owner:Object Reference) -> (then)
   > ReleaseResource(self:Fort Item Shop Offer VM Object Reference=@n1)
   > SetPresentationsLoadingState(LoadState=NewEnumerator0, ShouldClearPresentationsArray=true)
   > K2Node_Composite_0 = debug logs - reset tile!(execute) -> (then:Exec) {is=title K2Node_Composite}
  ]}
 debug logs - reset tile {
  path=ResetTile_Implementation.K2Node_Composite_0.debug logs - reset tile
  nodes [
   Inputs!() -> (execute)
   > n1 = Branch(Condition=Get EnableLogs())
   > n2 = Get()
   n3 = GetOwningWidgetName()
   n4 = GetTitle(self=@n2)
   > n5 = PrintString(InString=Append!(A="Reset tile for VM: ", B="To String (Text)"(InText=@n4), C=" - Object: ", D=@n3) -> (ReturnValue), bPrintToScreen=false)
   Outputs!(then:Exec=[@n6.then, @n5.then, @n1.else])
   n6 = PrintString(execute=@n2.else, InString=Append!(A="Reset tile - invalid VM - Object: ", B=@n3) -> (ReturnValue), bPrintToScreen=false)
  ]
 }
 ResumeOrPauseSwapTimer {nodes [
   n1 = ResumeOrPauseSwapTimer!() -> (then, ShouldResume:bool=~) {is=title K2Node_FunctionEntry}
   > n2 = Branch(Condition=@n1)
   n3 = Get PresentationSwapTimerHandle()
   > Branch(Condition=Is Timer Paused by Handle!(Handle:Timer Handle Structure=@n3) -> (ReturnValue:bool))
   > Unpause Timer by Handle!(execute, Handle:Timer Handle Structure=@n3) -> (then)
   Branch(execute=@n2.else, Condition=Is Timer Active by Handle!(Handle:Timer Handle Structure=@n3) -> (ReturnValue:bool))
   > Pause Timer by Handle!(execute, Handle:Timer Handle Structure=@n3) -> (then)
  ]}
 Set Row State Implementation {nodes [
   n1 = Set Row State Implementation!() -> (then, InRowState:E_ItemShop_RowState Enum=NewEnumerator0) {is=title K2Node_FunctionEntry}
   > n2 = Set RowState(RowState:E_ItemShop_RowState Enum=@n1) -> (Output_Get:E_ItemShop_RowState Enum=NewEnumerator4)
   > ResumeOrPauseSwapTimer!(execute, self, ShouldResume:bool="Equal (Enum)"(A:E_ItemShop_RowState Enum=@n2, B:E_ItemShop_RowState Enum=NewEnumerator0)) -> (then) {is=title K2Node_CallFunction}
   n3 = Get PresentationSwap()
   > StopAnimation(InAnimation=Select!(
     NewEnumerator4:Widget Animation Object Reference
     NewEnumerator1:Widget Animation Object Reference=@n3
     NewEnumerator0:Widget Animation Object Reference=Get Peek() -> (Peek:Widget Animation Object Reference)
     NewEnumerator2:Widget Animation Object Reference=@n3
     NewEnumerator3:Widget Animation Object Reference=@n3
     Index:E_ItemShop_RowState Enum=Get InRowState() -> (InRowState:E_ItemShop_RowState Enum=NewEnumerator4)
    ) -> (ReturnValue:Widget Animation Object Reference))
   > SetPeekAnimAlpha!(execute, self, Value:double=0) -> (then)
  ]}
 SetBackgroundLoadingState {nodes [
   n1 = SetBackgroundLoadingState!() -> (then, IsLoading:bool=~) {is=title K2Node_FunctionEntry}
   > Set IsLoadingBackground(IsLoadingBackground:bool=@n1) -> (Output_Get:bool)
   > UpdateTileLoadingState()
  ]}
 SetBackgroundMaterial {nodes [
   n1 = SetBackgroundMaterial!() -> (then, Material:Material Interface Object Reference) {is=title K2Node_FunctionEntry}
   > SetBrushFromMaterial(self=Get Background(), Material=@n1)
   > n2 = GetDynamicMaterial(self=Get Background())
   n3 = Get FortItemShopOfferVM()
   n4 = Get FilterVM!(self:Fort Item Shop VM Object Reference=Get FortItemShopVM() -> (FortItemShopVM:Fort Item Shop VM Object Reference)) -> (FilterVM:Fort Item Shop Filter VM Object Reference)
   > SetTileBackgroundColors!(
    execute
    MID:Material Instance Dynamic Object Reference=@n2
    FortItemShopOfferMetaDisplayData:"Fort Item Shop Offer Meta Display Data Structure (by ref)"=GetMetaDisplayDataForCurrentViewMode!(
     OfferMetaDisplayData:"Fort Item Shop Offer Meta Display Data Structure (by ref)"=Get MetaDisplayData!(self:Fort Item Shop Offer VM Object Reference=@n3) -> (MetaDisplayData:Fort Item Shop Offer Meta Display Data Structure)
     bIsFilterActive:bool=Get bIsActive!(self:Fort Item Shop Filter VM Object Reference=@n4) -> (bIsActive:bool)
    ) -> (ReturnValue:"Fort Item Shop Offer Meta Display Data Structure (by ref)")
   ) -> (then)
   > SetBackgroundLoadingState(IsLoading=false) {is=title K2Node_CallFunction}
   > SetScalarParameterValue(self=@n2, ParameterName=Seed, Value=RandomFloat!() -> (ReturnValue:double))
  ]}
 SetFrameBreakStyle {nodes [
   SetFrameBreakStyle!() -> (then, IsFrameBreaking:bool=~) {is=title K2Node_FunctionEntry}
   n1 = float * float(
    A=Get FrameBreakPadding!(self:Athena Item Shop Configuration Object Reference=GetItemShopConfiguration!() -> (ReturnValue:Athena Item Shop Configuration Object Reference)) -> (FrameBreakPadding:float)
    B=-1
   )
   > SetPadding!(
    execute
    self:Overlay Slot Object Reference=[SlotAsOverlaySlot(Widget=Get Image_Next()), SlotAsOverlaySlot(Widget=Get Image_Current())]
    InPadding:Margin Structure=Select!(
     Option 0:Margin Structure=Make Margin(Left=0, Top=0, Right=0, Bottom=0)
     Option 1:Margin Structure=Make Margin(Left=0, Top=@n1, Right=0, Bottom=0)
     Index:bool=Get IsFrameBreaking()
    ) -> (ReturnValue:Margin Structure)
   ) -> (then)
  ]}
 SetHoveredStateVisual {nodes [
   SetHoveredStateVisual!() -> (then, IsHovered:bool=~) {is=title K2Node_FunctionEntry}
   > n1 = GetDynamicMaterial(self=Get Image_Current())
   > Set OfferImageMID(OfferImageMID:Material Instance Dynamic Object Reference=@n1) -> (Output_Get:Material Instance Dynamic Object Reference)
   > n2 = GetDynamicMaterial(self=Get Image_Next())
   > Set NextOfferImageMID(NextOfferImageMID:Material Instance Dynamic Object Reference=@n2) -> (Output_Get:Material Instance Dynamic Object Reference)
   > n3 = Branch(Condition=Get IsHovered() -> (IsHovered:bool))
   > SetScalarParameterValue(
    self=[
     Get OfferImageMID() -> (OfferImageMID:Material Instance Dynamic Object Reference)
     Get NextOfferImageMID() -> (NextOfferImageMID:Material Instance Dynamic Object Reference)
    ]
    ParameterName=Hover
    Value=1
   )
   > SetScalarParameterValue(self=[Get OfferImageMID(), Get NextOfferImageMID()], ParameterName=HoverBurst, Value=1)
   SetScalarParameterValue(execute=@n3.else, self=[Get OfferImageMID(), Get NextOfferImageMID()], ParameterName=Hover, Value=0)
   > SetScalarParameterValue(self=[Get OfferImageMID(), Get NextOfferImageMID()], ParameterName=HoverBurst, Value=0)
  ]}
 SetLoadGuardState {nodes [
   SetLoadGuardState!() -> (then, State:E_ItemShop_LoadGuardState Enum=NewEnumerator0) {is=title K2Node_FunctionEntry}
   > SetBrushColor(
    self=Get Border_LoadGuard() -> (Border_LoadGuard:Border Object Reference)
    InBrushColor=Select!(
     NewEnumerator0:Linear Color Structure=(R=1,G=1,B=1,A=1)
     NewEnumerator1:Linear Color Structure=(R=1,G=1,B=1,A=0)
     NewEnumerator2:Linear Color Structure=(R=1,G=1,B=1,A=1)
     Index:E_ItemShop_LoadGuardState Enum=Get State() -> (State:E_ItemShop_LoadGuardState Enum=NewEnumerator0)
    ) -> (ReturnValue:Linear Color Structure)
   )
   > Set ContentColorAndOpacity!(
    execute
    ContentColorAndOpacity:Linear Color Structure=Select(
     NewEnumerator0=(R=1,G=1,B=1,A=0)
     NewEnumerator1=(R=1,G=1,B=1,A=1)
     NewEnumerator2=(R=1,G=1,B=1,A=0)
     Index=Get State() -> (State=NewEnumerator0)
    )
    self:Border Object Reference=Get Border_LoadGuard()
   ) -> (then, Output_Get:Linear Color Structure)
   > n1 = GetDynamicMaterial(self:Border Object Reference=Get Border_LoadGuard())
   > SetScalarParameterValue(
    self=@n1
    ParameterName=Is Empty
    Value=Select(NewEnumerator0:float=0, NewEnumerator1:float=0, NewEnumerator2:float=1, Index=Get State() -> (State=NewEnumerator0)) -> (ReturnValue:float)
   )
  ]}
 SetPeekAnimAlpha {nodes [
   SetPeekAnimAlpha!() -> (then, Value:double=~) {is=title K2Node_FunctionEntry}
   > n1 = Switch on E_ItemShop_RowState!(execute, Selection:E_ItemShop_RowState Enum=Get RowState() -> (RowState:E_ItemShop_RowState Enum=NewEnumerator4)) -> (NewEnumerator4:Exec, NewEnumerator1:Exec, NewEnumerator0:Exec, NewEnumerator2:Exec, NewEnumerator3:Exec) {is=title K2Node_SwitchEnum}
   SetRenderOpacity(
    execute=[@n1.NewEnumerator0, @n1.NewEnumerator2]
    self=[Get Image_Current(), Get Image_Next(), Get Background(), Get Overlay_All() -> (Overlay_All:Overlay Object Reference)]
    InOpacity=1
   )
   > SetRenderOpacity(self=Get Image_GhostTile() -> (Image_GhostTile:Image Object Reference), InOpacity=0)
   > n2 = GetDynamicMaterial(self:Image Object Reference=Get Image_Current())
   > SetScalarParameterValue(self=@n2, ParameterName=FadeImageAlpha, Value=0)
   SetRenderOpacity(execute=@n1.NewEnumerator3, self=Get Overlay_All(), InOpacity=1)
   > SetRenderOpacity(self=[Get Image_Next(), Get Background()], InOpacity=0)
   > SetRenderOpacity(self=Get Image_GhostTile(), InOpacity=Lerp(A=0.8, B=1, Alpha=Get Value() -> (Value:double)))
   > SetRenderOpacity(self=Get Image_Current(), InOpacity=Lerp(A=0.1, B=0.2, Alpha=Get Value()))
   > n3 = GetDynamicMaterial(self=Get Image_Current())
   > SetScalarParameterValue(self=@n3, ParameterName=FadeImageAlpha, Value=1)
   SetRenderOpacity(execute=@n1.NewEnumerator1, self=[Get Background(), Get Image_Current()], InOpacity=1)
   > SetRenderOpacity(self=[Get Image_Next(), Get Image_GhostTile()], InOpacity=0)
   > SetRenderOpacity(self=Get Overlay_All(), InOpacity=Lerp(A=0.3, B=0.4, Alpha=Get Value()))
   > n4 = GetDynamicMaterial(self=Get Image_Current())
   > SetScalarParameterValue(self=@n4, ParameterName=FadeImageAlpha, Value=1)
  ]}
 SetPeekState_Implementation {nodes [
   SetPeekState_Implementation!() -> (then, InIsPeeked:bool=~) {is=title K2Node_FunctionEntry}
   > n1 = Branch(Condition=IsUsingTouch(PlayerController=GetOwningPlayer()))
   n2 = Branch(execute=@n1.else, Condition=Get InIsPeeked() -> (InIsPeeked:bool))
   n3 = Get Peek()
   > PlayAnimationForward(InAnimation=@n3)
   PlayAnimationReverse(execute=@n2.else, InAnimation=@n3)
  ]}
 SetPresentation {nodes [
   n1 = SetPresentation!() -> (then, Image:Image Object Reference, Presentation:"Fort Item Shop Loaded Presentation Structure (by ref)") {is=title K2Node_FunctionEntry}
   > SetBrushFromMaterial(
    self=@n1
    Material=Get!() -> (Presentation_Texture:Texture 2D Object Reference, Presentation_Material:Material Interface Object Reference).Presentation_Material
   )
   > Is Valid(InputObject=Get())
   > n2 = GetDynamicMaterial(self=Get Image() -> (Image:Image Object Reference))
   > n3 = Set MID(MID:Material Instance Dynamic Object Reference=@n2) -> (Output_Get:Material Instance Dynamic Object Reference)
   > SetTextureParameterValue(self=@n3, ParameterName=OfferImage, Value=Get())
   n4 = Get FortItemShopOfferVM()
   n5 = Get FilterVM(self=Get FortItemShopVM())
   > SetTileBackgroundColors(
    MID=@n3
    FortItemShopOfferMetaDisplayData=GetMetaDisplayDataForCurrentViewMode(OfferMetaDisplayData=Get MetaDisplayData(self=@n4), bIsFilterActive=Get bIsActive(self=@n5))
   )
   > ApplyRulesetToMID!(
    execute
    self
    MID:Material Instance Dynamic Object Reference=Get MID()
    Rule:S Item Shop Image Ruleset Values Structure=Get ImageDisplayRuleset()
   ) -> (then) {is=title K2Node_CallFunction}
   > SetScalarParameterValue(
    self=Get MID()
    ParameterName=AddTopEdgeFade
    Value=Select!(Option 0:float=0, Option 1:float=1, Index:bool=Get IsFrameBreaking()) -> (ReturnValue)
   )
  ]}
 SetPresentationsLoadingState {nodes [
   n1 = SetPresentationsLoadingState!() -> (then, LoadState:E_ItemShop_LoadGuardState Enum=NewEnumerator0, ShouldClearPresentationsArray:bool=~) {is=title K2Node_FunctionEntry}
   > Set PresentationsLoadState(PresentationsLoadState:E_ItemShop_LoadGuardState Enum=@n1) -> (Output_Get:E_ItemShop_LoadGuardState Enum=NewEnumerator0)
   > n2 = Branch(Condition=@n1.ShouldClearPresentationsArray)
   > n3 = Clear(TargetArray:Array of Fort Item Shop Loaded Presentation Structures=Get LoadedPresentations() -> (LoadedPresentations:Array of Fort Item Shop Loaded Presentation Structures))
   UpdateTileLoadingState(execute=[@n3.then, @n2.else])
  ]}
 SetSinglePresentation {nodes [
   SetSinglePresentation!() -> (then, Presentation_Texture:Texture 2D Object Reference, Presentation_Material:Material Interface Object Reference) {is=title K2Node_FunctionEntry}
   > SetPresentation!(
    execute
    self
    Image:Image Object Reference=Get Image_Next()
    Presentation_Texture:Texture 2D Object Reference=/FortUILibrary/Input/PeripheralGuitar/Source/Empty1x1
    Presentation_Material:Material Interface Object Reference=Get DefaultPresentationMaterial()
   ) -> (then) {is=title K2Node_CallFunction}
   > SetPresentation!(
    execute
    self
    Image=Get Image_Current()
    Presentation:"Fort Item Shop Loaded Presentation Structure (by ref)"=Get Presentation() -> (Presentation:Fort Item Shop Loaded Presentation Structure)
   ) -> (then)
   > Reset Image Material Params() {is=title K2Node_CallFunction}
  ]}
 StartShowingPresentations {nodes [
   n1 = StartShowingPresentations!() -> (then, PresentationSwapDelegate:Delegate) {is=title K2Node_FunctionEntry}
   > n2 = Set PresentationsAmount(PresentationsAmount:int=Length(TargetArray=Get LoadedPresentations())) -> (Output_Get:int=0)
   > n3 = Switch on Int!(execute:Exec, Selection:int=@n2) -> (Default:Exec, 0:Exec, 1:Exec)
   SetPresentationsLoadingState(execute=@n3.1, LoadState=NewEnumerator1, ShouldClearPresentationsArray=false) {is=title K2Node_CallFunction}
   > SetSinglePresentation!(
    execute
    self
    Presentation:Fort Item Shop Loaded Presentation Structure="Get (a ref)"(Array:Array of Fort Item Shop Loaded Presentation Structures=Get LoadedPresentations(), Dimension 1=0) -> (Output:"Fort Item Shop Loaded Presentation Structure (by ref)")
   ) -> (then) {is=title K2Node_CallFunction}
   SetPresentationsLoadingState(execute=@n3.Default, LoadState=NewEnumerator1, ShouldClearPresentationsArray=false)
   > Set CurrentPresentationIndex(CurrentPresentationIndex=0) -> (Output_Get=0)
   > UpdatePresentations()
   > n4 = Set Timer by Event(Delegate=@n1, Time=Get TimeBetweenPresentations() -> (TimeBetweenPresentations:float), bLooping=true)
   > Set PresentationSwapTimerHandle(PresentationSwapTimerHandle:Timer Handle Structure=@n4) -> (Output_Get:Timer Handle Structure)
   > ResumeOrPauseSwapTimer(ShouldResume="Equal (Enum)"(A=Get RowState() -> (RowState=NewEnumerator4), B=NewEnumerator0))
   OnSetBrokenImage(execute=@n3.0) {is=title K2Node_CallFunction}
  ]}
 UpdatePresentations {nodes [
   UpdatePresentations() {is=title K2Node_FunctionEntry}
   > SetPresentation(
    Image=Get Image_Current()
    Presentation="Get (a ref)"(Array=Get LoadedPresentations(), Dimension 1=Get CurrentPresentationIndex() -> (CurrentPresentationIndex=0))
   )
   > SetPresentation(
    Image=Get Image_Next()
    Presentation="Get (a ref)"(Array=Get LoadedPresentations(), Dimension 1=GetNextPresentationIndex() -> (ReturnValue=0) {is=title K2Node_CallFunction})
   )
   > Reset Image Material Params()
  ]}
 UpdateRulesetDebugText {nodes [
   n1 = UpdateRulesetDebugText!() -> (then, Ruleset:String) {is=title K2Node_FunctionEntry}
   > Set Text(Text="To Text (String)"(InString=@n1), self=Get Debug_Ruleset())
  ]}
 UpdateTileLoadingState {nodes [
   UpdateTileLoadingState() {is=title K2Node_FunctionEntry}
   > SetLoadGuardState!(
    execute
    self
    State:E_ItemShop_LoadGuardState Enum=Select(
     Option 0:E_ItemShop_LoadGuardState Enum=Get PresentationsLoadState() -> (PresentationsLoadState:E_ItemShop_LoadGuardState Enum=NewEnumerator0)
     Option 1:E_ItemShop_LoadGuardState Enum=NewEnumerator0
     Index=Get IsLoadingBackground() -> (IsLoadingBackground:bool)
    ) -> (ReturnValue:E_ItemShop_LoadGuardState Enum)
   ) -> (then) {is=title K2Node_CallFunction}
  ]}
 VM_DisplayAssetPathChanged {nodes [
   n1 = VM_DisplayAssetPathChanged!() -> (then, DisplayAssetPath:String) {is=title K2Node_FunctionEntry}
   > Branch(Condition=ShouldLoadPresentationsOnTileInit!(self:Fort Item Shop Offer VM Object Reference=Get FortItemShopOfferVM()) -> (ReturnValue:bool))
   > Set with Broadcast DisplayAssetPath_BP(DisplayAssetPath_BP=@n1)
  ]}
 VM_GetImageDisplayRuleset {nodes [
   n1 = VM_GetImageDisplayRuleset!() -> (
    then
    SecondaryGrantedItems:Array of Fort Item VM Object References
    IsFilterActive:bool=~
    TileSize:EItemShopTileSize Enum=Size_1_x_1
    OfferType:EFortItemShopOfferType Enum=LevelBundleOffer
    UsingFullTileOverridePresentations:bool=~
   ) {is=title K2Node_FunctionEntry}
   > n2 = GetImageDisplayRuleset!(
    execute
    OfferVM:Fort Item Shop Offer VM Object Reference=Get FortItemShopOfferVM()
    ItemShopVM:Fort Item Shop VM Object Reference=Get FortItemShopVM()
    ItemsToShow:Array of Fort Item VM Object References=Select(Option 0:Array of Fort Item VM Object References=@n1, Option 1:Array of Fort Item VM Object References, Index=@n1.IsFilterActive) -> (ReturnValue:Array of Fort Item VM Object References)
    IsFilteringActive:bool=@n1.IsFilterActive
   ) -> (then, S_ItemShop_ImageRulesetValues:S Item Shop Image Ruleset Values Structure, RulesetName:String)
   > UpdateRulesetDebugText!(execute, self, Ruleset:String=@n2.RulesetName) -> (then) {is=title K2Node_CallFunction}
   > Return Node!(execute, S_ItemShop_ImageRulesetValues:S Item Shop Image Ruleset Values Structure=@n2)
  ]}
 VM_MetaDisplayDataUpdated {nodes [
   VM_MetaDisplayDataUpdated!() -> (then, MetaDisplayData:Fort Item Shop Offer Meta Display Data Structure) {is=title K2Node_FunctionEntry}
   > SetBackgroundLoadingState(IsLoading=true)
   > n1 = LoadBackground!(
    execute
    self:Fort Item Shop Offer VM Object Reference=Get FortItemShopOfferVM()
    Owner:Object Reference
    BackgroundKey:String=Break Fort Item Shop Offer Meta Display Data!(FortItemShopOfferMetaDisplayData:"Fort Item Shop Offer Meta Display Data Structure (by ref)"=Get MetaDisplayData!() -> (MetaDisplayData)) -> (TileBackgroundMaterialKey:String) {is=title K2Node_BreakStruct}
    OnLoadedCallback:"Delegate (by ref)"=Create Event()
   ) -> (then, ReturnValue:bool)
   > n2 = Branch(Condition=@n1)
   SetBackgroundMaterial(execute=@n2.else, Material=Get DefaultBackgroundGradientMaterial()) {is=title K2Node_CallFunction}
  ]}
 VM_PresentationDataUpdated {nodes [
   VM_PresentationDataUpdated!() -> (then, DisplayAssetPath:String) {is=title K2Node_FunctionEntry}
   n1 = Get FortItemShopOfferVM()
   n2 = Get FilterVM(self=Get FortItemShopVM())
   > VM_UpdateImageDisplayRuleset!(
    execute
    self
    ImageDisplayRuleset:S Item Shop Image Ruleset Values Structure=VM_GetImageDisplayRuleset!(
     self
     SecondaryGrantedItems:Array of Fort Item VM Object References=Get DisplayedSecondaryGrantedItems!(self:Fort Item Shop Offer VM Object Reference=@n1) -> (DisplayedSecondaryGrantedItems:Array of Fort Item VM Object References)
     IsFilterActive:bool=Get bIsActive(self=@n2)
     TileSize:EItemShopTileSize Enum=Get TileSize(self=@n1) -> (TileSize=Size_1_x_1)
     OfferType:EFortItemShopOfferType Enum=Get OfferType!(self:Fort Item Shop Offer VM Object Reference=@n1) -> (OfferType:EFortItemShopOfferType Enum=None)
     UsingFullTileOverridePresentations:bool=Get bUsingFullTileOverridePresentations!(self:Fort Item Shop Offer VM Object Reference=@n1) -> (bUsingFullTileOverridePresentations:bool)
    ) -> (S_ItemShop_ImageRulesetValues:S Item Shop Image Ruleset Values Structure) {is=title K2Node_CallFunction}
   ) -> (then)
   n3 = Get LoadedPresentations()
   n4 = Get bUsingFullTileOverridePresentations(self=Get FortItemShopOfferVM())
   > Branch(Condition=OR Boolean!(
     A=ForceLoadPresentationsForFullTileOverrides!() -> (ReturnValue:bool)
     B=NOT Boolean(A=@n4)
     C:bool=Is Empty!(TargetArray:Array of Fort Item Shop Loaded Presentation Structures=@n3) -> (ReturnValue:bool)
    ) -> (ReturnValue))
   > LoadPresentations!(execute, self:Self Object Reference) -> (then)
  ]}
 VM_UpdateImageDisplayRuleset {nodes [
   n1 = VM_UpdateImageDisplayRuleset!() -> (then, ImageDisplayRuleset:S Item Shop Image Ruleset Values Structure) {is=title K2Node_FunctionEntry}
   > Set ImageDisplayRuleset(ImageDisplayRuleset:S Item Shop Image Ruleset Values Structure=@n1) -> (Output_Get:S Item Shop Image Ruleset Values Structure)
   > n2 = GetDynamicMaterial(self=Get Image_Current())
   > ApplyRulesetToMID(MID=@n2, Rule=Get ImageDisplayRuleset())
   > n3 = GetDynamicMaterial(self=Get Image_Next())
   > ApplyRulesetToMID(MID=@n3, Rule=Get ImageDisplayRuleset())
  ]}
]
variables [
 CurrentPresentationIndex {type=integer}
 DefaultBackgroundGradientMaterial {type=MaterialInterface, default=/FortUILibrary/Systems/ItemShopTile/Materials/M_UI_ItemShopTile_Background_Gradient}
 DefaultBackgroundTextureMaterial {type=MaterialInterface, default=/FortUILibrary/Systems/ItemShopTile/Materials/M_UI_ItemShopTile_Background_Texture}
 DefaultPresentationMaterial {type=MaterialInterface, default=/FortUILibrary/Systems/ItemShopTile/Materials/M_UI_ItemShopTile_OfferImage_Default}
 DisplayAssetPath_BP {type=string, default=""}
 EnableLogs {type=boolean}
 ExpectedSoftPath {type=SoftObjectPath, default=""}
 ImageDisplayRuleset {
  type=S_ItemShop_ImageRulesetValues
  default {hasRoundedCorners=false, scaleToFitX=false, zoom=0, zoom_FrameBreaking=0, position=<0 0>, position_FrameBreaking=<0 0>}
 }
 Is Hovered {type=boolean}
 IsFrameBreaking {type=boolean}
 IsLoadingBackground {type=boolean}
 IsPreviewingFrameBreak {type=boolean}
 LoadedPresentations {type="array<FortItemShopLoadedPresentation>", default []}
 PeekAnimAlpha {type=number}
 PresentationsAmount {type=integer}
 PresentationsLoadState {type=E_ItemShop_LoadGuardState, default=Loading}
 PresentationSwapTimerHandle {type=TimerHandle}
 RowState {type=E_ItemShop_RowState, default=None}
 SectionID {type=string, default=""}
 TileDimensions {type=Vector2D, default=<0 0>}
 TimeBetweenPresentations {type=number, default=3}
]
widgets {
 info {inheritedWidgetCount=0, parentClass=&UserWidget, rootWidgetClass=&Border, widgetCount=8}
 widgets [
  Border_LoadGuard {
   widgetClassPath=&Border
   bIsVariable=true
   widget=Border_LoadGuard
   properties {
    background {
     bIsDynamicallyLoaded=false
     drawAs=Image
     imageType=FullColor
     mirroring=NoMirror
     resourceName=None
     resourceObject=/FortUILibrary/Systems/ItemShopTile/Materials/MI_UI_Tile_LoadGuard
     tiling=NoTile
     imageSize=<32 32>
     margin=<ltrb 0 0 0 0>
     outlineSettings {
      bUseBrushTransparency=false
      roundingType=HalfHeightRadius
      width=0
      color {colorUseRule=UseColor_Specified, specifiedColor=<rgba 0 0 0 0>}
      cornerRadii=<0 0 0 0>
     }
     tintColor {colorUseRule=UseColor_Specified, specifiedColor=<rgba 1 1 1 1>}
     uVRegion {bIsValid=false, max=<0 0>, min=<0 0>}
    }
    padding=<ltrb 0 0 0 0>
   }
  }
  Overlay_All {
   widgetClassPath=&Overlay
   bIsVariable=true
   parent=Border_LoadGuard
   widget=Overlay_All
   slot {class=BorderSlot, properties {padding=<ltrb 0 0 0 0>}}
  }
  Background {
   widgetClassPath=&Image
   bIsVariable=true
   parent=Overlay_All
   widget=Background
   properties {
    visibility=SelfHitTestInvisible
    brush {
     bIsDynamicallyLoaded=false
     drawAs=Image
     imageType=FullColor
     mirroring=NoMirror
     resourceName=None
     resourceObject=&M_UI_ItemShopTile_Background_Gradient
     tiling=NoTile
     imageSize=<32 32>
     margin=<ltrb 0 0 0 0>
     outlineSettings {
      bUseBrushTransparency=false
      roundingType=HalfHeightRadius
      width=0
      color {colorUseRule=UseColor_Specified, specifiedColor=<rgba 0 0 0 0>}
      cornerRadii=<0 0 0 0>
     }
     tintColor {colorUseRule=UseColor_Specified, specifiedColor=<rgba 1 1 1 1>}
     uVRegion {bIsValid=false, max=<0 0>, min=<0 0>}
    }
   }
   slot {class=OverlaySlot, properties {horizontalAlignment=HAlign_Fill, verticalAlignment=VAlign_Fill}}
  }
  ModularBlock_Outline {
   widgetClassPath=/FortUILibrary/UIKit/Buttons/Blocks/WBP_UIKit_Block_Outline
   bIsVariable=true
   parent=Overlay_All
   widget=ModularBlock_Outline
   properties {bottomLeftCornerRadius=18, bottomRightCornerRadius=18, radiusInPixels=true, topLeftCornerRadius=18, topRightCornerRadius=18}
   slot {class=OverlaySlot, properties {horizontalAlignment=HAlign_Fill, verticalAlignment=VAlign_Fill}}
  }
  Image_Next {
   widgetClassPath=&Image
   bIsVariable=true
   parent=Overlay_All
   widget=Image_Next
   properties {
    visibility=SelfHitTestInvisible
    brush {
     bIsDynamicallyLoaded=false
     drawAs=Image
     imageType=FullColor
     mirroring=NoMirror
     resourceName=None
     resourceObject=&M_UI_ItemShopTile_OfferImage_Default
     tiling=NoTile
     imageSize=<32 32>
     margin=<ltrb 0 0 0 0>
     outlineSettings {
      bUseBrushTransparency=false
      roundingType=HalfHeightRadius
      width=0
      color {colorUseRule=UseColor_Specified, specifiedColor=<rgba 0 0 0 0>}
      cornerRadii=<0 0 0 0>
     }
     tintColor {colorUseRule=UseColor_Specified, specifiedColor=<rgba 1 1 1 1>}
     uVRegion {bIsValid=false, max=<0 0>, min=<0 0>}
    }
   }
   slot {class=OverlaySlot, properties {horizontalAlignment=HAlign_Fill, verticalAlignment=VAlign_Fill}}
  }
  Image_Current {
   widgetClassPath=&Image
   bIsVariable=true
   parent=Overlay_All
   widget=Image_Current
   properties {
    visibility=SelfHitTestInvisible
    brush {
     bIsDynamicallyLoaded=false
     drawAs=Image
     imageType=FullColor
     mirroring=NoMirror
     resourceName=None
     resourceObject=&M_UI_ItemShopTile_OfferImage_Default
     tiling=NoTile
     imageSize=<32 32>
     margin=<ltrb 0 0 0 0>
     outlineSettings {
      bUseBrushTransparency=false
      roundingType=HalfHeightRadius
      width=0
      color {colorUseRule=UseColor_Specified, specifiedColor=<rgba 0 0 0 0>}
      cornerRadii=<0 0 0 0>
     }
     tintColor {colorUseRule=UseColor_Specified, specifiedColor=<rgba 1 1 1 1>}
     uVRegion {bIsValid=false, max=<0 0>, min=<0 0>}
    }
   }
   slot {class=OverlaySlot, properties {horizontalAlignment=HAlign_Fill, verticalAlignment=VAlign_Fill}}
  }
  Image_GhostTile {
   widgetClassPath=&Image
   bIsVariable=true
   parent=Overlay_All
   widget=Image_GhostTile
   properties {
    renderOpacity=0
    visibility=SelfHitTestInvisible
    brush {
     bIsDynamicallyLoaded=false
     drawAs=Image
     imageType=FullColor
     mirroring=NoMirror
     resourceName=None
     resourceObject=/FortUILibrary/Systems/ItemShopTile/Materials/MI_UI_ItemShop_GhostTile
     tiling=NoTile
     imageSize=<32 32>
     margin=<ltrb 0 0 0 0>
     outlineSettings {
      bUseBrushTransparency=false
      roundingType=HalfHeightRadius
      width=0
      color {colorUseRule=UseColor_Specified, specifiedColor=<rgba 0 0 0 0>}
      cornerRadii=<0 0 0 0>
     }
     tintColor {colorUseRule=UseColor_Specified, specifiedColor=<rgba 1 1 1 1>}
     uVRegion {bIsValid=false, max=<0 0>, min=<0 0>}
    }
   }
   slot {class=OverlaySlot, properties {horizontalAlignment=HAlign_Fill, verticalAlignment=VAlign_Fill}}
  }
  Debug_Ruleset {
   widgetClassPath=&CommonTextBlock
   bIsVariable=true
   parent=Overlay_All
   widget=Debug_Ruleset
   properties {
    text=Ruleset debug
    visibility=Collapsed
    font {
     bForceMonospaced=false
     bMaterialIsStencil=false
     fontMaterial=null
     fontObject=&NotoSans
     letterSpacing=0
     monospacedWidth=1
     size=21
     skewAmount=0
     typefaceFontName=NotoSans-Regular
     outlineSettings {
      bApplyOutlineToDropShadows=false
      bMiteredCorners=false
      bSeparateFillAlpha=false
      outlineMaterial=null
      outlineSize=2
      outlineColor=<rgba 0 0 0 1>
     }
    }
    shadowOffset=<0 0>
   }
   slot {class=OverlaySlot, properties {horizontalAlignment=HAlign_Center, verticalAlignment=VAlign_Center}}
  }
 ]
}
